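# M_UI_ScreenBG is a UI-domain material, named as a screen background for the ArmTale frontend.
# Its only node is a call to the shared material function MF_UI_BackgroundFX, which holds the effect logic outside this file.
# The call carries GridTiling, GridOpacity and GridFlickerSpeed, which point to a tiled, flickering grid effect.
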Material /ArmTaleFrontendUI/Assets/Materials/M_UI_ScreenBG
graph {
 properties {bAutomaticallySetUsageInEditor=false, bUsedWithStaticMesh=false, materialDomain=MD_UI, usedShadingModels=Default Lit}
 nodes [MaterialFunctionCall!(GridTiling, GridOpacity, GridFlickerSpeed) {materialFunction=/ArmTaleFrontendUI/Shared/MF_UI_BackgroundFX}]
}
