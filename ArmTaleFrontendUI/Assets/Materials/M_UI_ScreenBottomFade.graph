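# Translucent UI material for the ArmTale frontend that, by its name, provides a fade at the bottom of the screen.
# It masks the vertical component of normalized UI UVs, scales it slightly, saturates it and applies a SmoothStep to form a soft gradient that rises toward the bottom.
# The graph also contains the shared MF_UI_BackgroundFX function, which exposes grid tiling, opacity and flicker speed inputs.
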Material /ArmTaleFrontendUI/Assets/Materials/M_UI_ScreenBottomFade
graph {
 properties {
  bAutomaticallySetUsageInEditor=false
  bUsedWithStaticMesh=false
  blendMode=BLEND_Translucent
  materialDomain=MD_UI
  usedShadingModels=Default Lit
 }
 nodes [
  MaterialFunctionCall!(GridTiling, GridOpacity, GridFlickerSpeed) {materialFunction=/ArmTaleFrontendUI/Shared/MF_UI_BackgroundFX}
  n1 = ComponentMask(None=MaterialFunctionCall() {materialFunction=&GetUserInterfaceUV}.Normalized UV) {g=true}
  SmoothStep(Value=Saturate(None=ConstantBiasScale(None=@n1) {bias=0, scale=1.1}))
 ]
}
