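# Translucent UI material for a weapon background glow, building a radial gradient between two rarity color parameters that default to blues.
# A circle SDF function supplies distance, which a SmoothStep with reversed bounds turns into an inverted falloff.
# A quartic ease-in curve shapes that falloff, which then drives the blend from Rarity_2 toward Rarity_1.
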Material /ArmTaleFrontendUI/Assets/Materials/M_UI_WeaponBGGlow
graph {
 properties {
  bAutomaticallySetUsageInEditor=false
  bUsedWithStaticMesh=false
  blendMode=BLEND_Translucent
  materialDomain=MD_UI
  usedShadingModels=Default Lit
 }
 nodes [
  n1 = MaterialFunctionCall!(UVs, Radius=Constant()) {materialFunction=/FortUILibrary/MaterialFunctions/Shapes/SDFs/MF_SDF_Circle}
  LinearInterpolate(
   A=VectorParameter() {parameterName=Rarity_2, defaultValue=<rgba 0 0.0595 0.2542 1>}.RGB
   B=VectorParameter() {parameterName=Rarity_1, defaultValue=<rgba 0 0.4287 1 1>}.RGB
   Alpha=MaterialFunctionCall!("Value (0-1)"=SmoothStep(Value=@n1.SDF) {constMax=0, constMin=1}) {materialFunction=/FortUILibrary/MaterialFunctions/Basics/MF_UI_EaseCurves_04_Quart}.In
  )
 ]
}
