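# List entry button widget that shows one weapon mod as an icon tile in the ArmTale frontend mod list.
# The assigned list item is cast and stored as the VM_ArmTale_ModEntry_FAKE view model.
# Handler functions feed the tile's dynamic material with equipped and empty states, the mod's lazy-loaded preview icon and its rarity colors.
# Hover and focus animations drive material highlight and outline parameters, with hover skipped while selected.
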WidgetBlueprint /ArmTaleFrontendUI/Widgets/WBP_ArmTale_ModListEntry
animations [
 AnimHovered {
  binding_count=1
  length_seconds=0.15
  object=&self:AnimHovered
  sequence {bindings [Image_ModTile {
     id=#1
     possessed_class=FortLazyImage
     tracks [{
       type=MovieSceneWidgetMaterialTrack
       sections [{
         type=MovieSceneParameterSection
         end_frame=3
         start_frame=0
         channels [
          Highlighted {type=MovieSceneScriptingFloatChannel, keys [[0,0],[3,0.1]]}
          Tile_OutlineThickness {type=MovieSceneScriptingFloatChannel, keys [[0,0.02],[3,0.04]]}
         ]
        }]
      }]
    }]}
 }
 AnimFocused {
  binding_count=1
  length_seconds=0.15
  object=&self:AnimFocused
  sequence {bindings [Image_ModTile {
     id=#2
     possessed_class=FortLazyImage
     tracks [{
       type=MovieSceneWidgetMaterialTrack
       sections [{
         type=MovieSceneParameterSection
         end_frame=3
         start_frame=0
         channels [
          Highlighted {type=MovieSceneScriptingFloatChannel, keys [[0,0],[3,1]]}
          Tile_OutlineThickness {type=MovieSceneScriptingFloatChannel, keys [[0,0.02],[3,0.05]]}
         ]
        }]
      }]
    }]}
 }
]
graphs [
 EventGraph {nodes [
   Event On Hovered()
   n1 = GetSelected!(self:Common Button Base Object Reference) -> (ReturnValue:bool)
   > n2 = Branch(Condition=@n1)
   n3 = Get AnimHovered() -> (AnimHovered:Widget Animation Object Reference)
   PlayAnimationForward(execute=@n2.else, InAnimation=@n3)
   n4 = Event OnListItemObjectSet()
   > n5 = Cast To VM_ArmTale_ModEntry_FAKE!(execute, Object=@n4.ListItemObject) -> (then, CastFailed, AsVM Arm Tale Mod Entry FAKE:VM Arm Tale Mod Entry FAKE Object Reference) {is=title K2Node_DynamicCast}
   > Set VM_ArmTale_ModEntry_FAKE(VM_ArmTale_ModEntry_FAKE:VM Arm Tale Mod Entry FAKE Object Reference=@n5) -> (Output_Get:VM Arm Tale Mod Entry FAKE Object Reference)
   Event On Unhovered()
   > n6 = Branch(Condition=@n1)
   PlayAnimationReverse(execute=@n6.else, InAnimation=@n3)
   Event On Focused() {is=title K2Node_Event}
   n7 = Get AnimFocused() -> (AnimFocused:Widget Animation Object Reference)
   > PlayAnimationForward(InAnimation=@n7)
   Event On Unfocused() {is=title K2Node_Event}
   > PlayAnimationReverse(InAnimation=@n7)
  ]}
 HandleIsEquippedChanged {nodes [
   n1 = HandleIsEquippedChanged!() -> (then, IsEquipped:bool=~) {is=title K2Node_FunctionEntry}
   > n2 = GetDynamicMaterial(self=Get Image_ModTile() -> (Image_ModTile:Fort Lazy Image Object Reference))
   > SetScalarParameterValue(self=@n2, ParameterName=IsEquipped, Value="To Float (Boolean)"(InBool=@n1))
  ]}
 HandleModDataChanged {nodes [
   HandleModDataChanged!() -> (then, ModData:S Arm Tale Mod Data FAKE Structure) {is=title K2Node_FunctionEntry}
   > n1 = GetDynamicMaterial(self=Get Image_ModTile())
   > Set ModTileMaterial(ModTileMaterial:Material Instance Dynamic Object Reference=@n1) -> (Output_Get:Material Instance Dynamic Object Reference)
   > n2 = Is Valid(InputObject=Get!() -> (
     ModData_ItemDefinition_34_#3:Fort Weapon Mod Item Definition Object Reference
     ModData_ModType_30_#4:E_ArmTale_ModTypes_FAKE Enum
     ModData_Stat0Buff_26_#5:double=~
     ModData_Stat1Buff_20_#6:double=~
     ModData_Stat2Buff_22_#7:double=~
     ModData_Stat3Buff_24_#8:double=~
    ))
   n3 = Get ModTileMaterial() -> (ModTileMaterial:Material Instance Dynamic Object Reference)
   SetScalarParameterValue(execute=@n2.Is Not Valid, self=@n3, ParameterName=IsEmpty, Value=1)
   SetScalarParameterValue(execute=@n2.Is Valid, self=@n3, ParameterName=IsEmpty, Value=0)
   n4 = Get() -> (
    ModData_Stat0Buff_26_#5=~
    ModData_Stat1Buff_20_#6=~
    ModData_Stat2Buff_22_#7=~
    ModData_Stat3Buff_24_#8=~
   )
   > SetBrushFromLazyTexture!(
    execute
    self:Common Lazy Image Object Reference=Get Image_ModTile()
    LazyTexture:"Texture 2D Soft Object Reference (by ref)"=GetLargePreviewImage!(ItemDefinition:Item Definition Base Object Reference=@n4) -> (ReturnValue:Texture 2D Soft Object Reference)
    bMatchSize:bool=false
   ) -> (then)
   > n5 = GetRarityData!(
    execute
    Rarity:EFortRarity Enum=GetRarity!(ItemDefinition:Item Definition Base Object Reference=@n4) -> (ReturnValue:EFortRarity Enum=Common)
   ) -> (
    then
    ReturnValue_Name:Text
    ReturnValue_Color1:Linear Color Structure
    ReturnValue_Color2:Linear Color Structure
    ReturnValue_Color3:Linear Color Structure
    ReturnValue_Color4:Linear Color Structure
    ReturnValue_Color5:Linear Color Structure
    ReturnValue_Radius:float=~
    ReturnValue_Falloff:float=~
    ReturnValue_Brightness:float=~
    ReturnValue_Roughness:float=~
    ReturnValue_Glow:float=~
   )
   n6 = Get ModTileMaterial()
   > SetVectorParameterValue(self=@n6, ParameterName=RarityColor_1, Value=@n5.ReturnValue_Color1)
   > SetVectorParameterValue(self=@n6, ParameterName=RarityColor_2, Value=@n5.ReturnValue_Color2)
   > SetVectorParameterValue(self=@n6, ParameterName=RarityColor_3, Value=@n5.ReturnValue_Color3)
  ]}
]
variables []
widgets {
 info {inheritedWidgetCount=2, parentClass=/Script/FortniteUI.FortItemListEntryButton, rootWidgetClass=&Overlay, widgetCount=4}
 widgets [
  Overlay_17 {widgetClassPath=&Overlay, widget=Overlay_17}
  Image_ModTile {
   widgetClassPath=&FortLazyImage
   bIsVariable=true
   parent=Overlay_17
   widget=Image_ModTile
   properties {
    bOverride_Cursor=true
    materialTextureParamName=IconTexture
    visibility=Visible
    brush {
     bIsDynamicallyLoaded=false
     drawAs=Image
     imageType=FullColor
     mirroring=NoMirror
     resourceName=None
     resourceObject=/ArmTaleFrontendUI/Assets/Materials/MI_UI_ModTile_ListEntry
     tiling=NoTile
     imageSize=<110 110>
     margin=<ltrb 0 0 0 0>
     outlineSettings {
      bUseBrushTransparency=false
      roundingType=HalfHeightRadius
      width=0
      color {colorUseRule=UseColor_Specified, specifiedColor=<rgba 0 0 0 0>}
      cornerRadii=<0 0 0 0>
     }
     tintColor {colorUseRule=UseColor_Specified, specifiedColor=<rgba 1 1 1 1>}
     uVRegion {bIsValid=false, max=<0 0>, min=<0 0>}
    }
   }
   slot {class=OverlaySlot, properties {horizontalAlignment=HAlign_Fill, verticalAlignment=VAlign_Fill}}
  }
  ItemWidget {widgetClassPath=/Script/FortniteUI.FortMultiSizeItemCard, bInherited=true, widget=null}
  InputActionWidget {widgetClassPath=&CommonActionWidget, bInherited=true, widget=null}
 ]
}
