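# CommonActivatableWidget panel for picking a weapon mod from a tile grid, backed by the placeholder VM_ArmTale_FAKE viewmodel.
# Clicking a tile equips that mod on the selected weapon, while hover sets or clears the viewmodel's preview mod.
# HandleModListEntriesChanged refills the tile view, finds the equipped mod by item definition and navigates to the stored index, which persists when nothing matches.
# The tile view is the desired focus target.
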WidgetBlueprint /ArmTaleFrontendUI/Widgets/WBP_ArmTale_ModSelectionPanel
graphs [
 BP_GetDesiredFocusTarget {nodes [
   Get Desired Focus Target()
   > Return Node!(execute, ReturnValue:Widget Object Reference=Get TileView_ModList() -> (TileView_ModList:Fort Tile View Object Reference))
  ]}
 EventGraph {nodes [
   n1 = "On Item Clicked (TileView_ModList)"!() -> (OutputDelegate, then, Item:Object Reference) {is=title K2Node_ComponentBoundEvent}
   > n2 = Cast To VM_ArmTale_ModEntry_FAKE!(execute, Object=@n1.Item) -> (then, CastFailed, AsVM Arm Tale Mod Entry FAKE:VM Arm Tale Mod Entry FAKE Object Reference) {is=title K2Node_DynamicCast}
   > EquipModToSelectedWeapon!(
    execute
    self:VM Arm Tale FAKE Object Reference=Get VM_ArmTale_FAKE() -> (VM_ArmTale_FAKE:VM Arm Tale FAKE Object Reference)
    ModToEquip:VM Arm Tale Mod Entry FAKE Object Reference=@n2
   ) -> (then)
   n3 = "On Item Is Hovered Changed (TileView_ModList)"!() -> (OutputDelegate, then, Item:Object Reference, bIsHovered:bool) {is=title K2Node_ComponentBoundEvent}
   > n4 = Branch(Condition=@n3.bIsHovered)
   > n5 = Cast To VM_ArmTale_ModEntry_FAKE(Object=@n3.Item)
   > Set with Broadcast PreviewModForSelectedWeapon!(
    execute
    PreviewModForSelectedWeapon:VM Arm Tale Mod Entry FAKE Object Reference=@n5
    self:VM Arm Tale FAKE Object Reference=Get VM_ArmTale_FAKE()
   ) -> (then, Output_Get:VM Arm Tale Mod Entry FAKE Object Reference)
   Set with Broadcast PreviewModForSelectedWeapon(execute=@n4.else, self=Get VM_ArmTale_FAKE())
  ]}
 HandleModListEntriesChanged {nodes [
   n1 = HandleModListEntriesChanged!() -> (then, ModListEntries:Array of VM Arm Tale Mod Entry FAKE Object References) {is=title K2Node_FunctionEntry}
   > Set List Items!(execute, self:List View Object Reference=Get TileView_ModList(), InListItems:Array of Object References=@n1) -> (then) {is=title K2Node_CallFunction}
   > n2 = GetEquippedModForSelectedSlot!(execute, self:VM Arm Tale FAKE Object Reference=Get VM_ArmTale_FAKE()) -> (then, EquippedMod:VM Arm Tale Mod Entry FAKE Object Reference)
   > n3 = Is Valid(InputObject=@n2)
   NavigateToIndex!(
    execute=[@n4.Completed, @n3.Is Not Valid]
    self:List View Object Reference=Get TileView_ModList()
    Index:int=Get EquippedEntryIndex() -> (EquippedEntryIndex:int=0)
   ) -> (then)
   Set EquippedModData(
    execute=@n3.Is Valid
    EquippedModData:S Arm Tale Mod Data FAKE Structure=Get ModData!(self:VM Arm Tale Mod Entry FAKE Object Reference=@n2) -> (ModData:S Arm Tale Mod Data FAKE Structure)
   ) -> (Output_Get:S Arm Tale Mod Data FAKE Structure)
   > n4 = For Each Loop with Break!(
    Exec:Exec
    Array:Array of VM Arm Tale Mod Entry FAKE Object References=Get ModListEntries() -> (ModListEntries:Array of VM Arm Tale Mod Entry FAKE Object References)
    Break:Exec=@n5.then
   ) -> (LoopBody:Exec, Array Element:VM Arm Tale Mod Entry FAKE Object Reference, Array Index:int, Completed:Exec)
   > Branch(Condition="Equal (Object)"(
     A=Get!(self:VM Arm Tale Mod Entry FAKE Object Reference=@n4) -> (
      ModData_ItemDefinition_34_#1:Fort Weapon Mod Item Definition Object Reference
      ModData_ModType_30_#2:E_ArmTale_ModTypes_FAKE Enum
      ModData_Stat0Buff_26_#3:double=~
      ModData_Stat1Buff_20_#4:double=~
      ModData_Stat2Buff_22_#5:double=~
      ModData_Stat3Buff_24_#6:double=~
     )
     B=Get!() -> (
      EquippedModData_ItemDefinition_34_#1:Fort Weapon Mod Item Definition Object Reference
      EquippedModData_ModType_30_#2:E_ArmTale_ModTypes_FAKE Enum
      EquippedModData_Stat0Buff_26_#3:double=~
      EquippedModData_Stat1Buff_20_#4:double=~
      EquippedModData_Stat2Buff_22_#5:double=~
      EquippedModData_Stat3Buff_24_#6:double=~
     )
    ))
   > n5 = Set EquippedEntryIndex(EquippedEntryIndex:int=@n4.Array Index) -> (Output_Get:int=0)
  ]}
]
variables []
widgets {
 info {inheritedWidgetCount=0, parentClass=&CommonActivatableWidget, rootWidgetClass=&Overlay, widgetCount=4}
 widgets [
  Overlay_0 {widgetClassPath=&Overlay, widget=Overlay_0}
  Image_PanelBackground {
   widgetClassPath=&Image
   parent=Overlay_0
   widget=Image_PanelBackground
   properties {brush {
     bIsDynamicallyLoaded=false
     drawAs=Image
     imageType=FullColor
     mirroring=NoMirror
     resourceName=None
     resourceObject=/ArmTaleFrontendUI/Assets/Materials/MI_UI_InfoPanelBG
     tiling=NoTile
     imageSize=<32 32>
     margin=<ltrb 0 0 0 0>
     outlineSettings {
      bUseBrushTransparency=false
      roundingType=HalfHeightRadius
      width=0
      color {colorUseRule=UseColor_Specified, specifiedColor=<rgba 0 0 0 0>}
      cornerRadii=<0 0 0 0>
     }
     tintColor {colorUseRule=UseColor_Specified, specifiedColor=<rgba 1 1 1 1>}
     uVRegion {bIsValid=false, max=<0 0>, min=<0 0>}
    }}
   slot {class=OverlaySlot, properties {horizontalAlignment=HAlign_Fill, verticalAlignment=VAlign_Fill}}
  }
  SizeBox_0 {
   widgetClassPath=&SizeBox
   parent=Overlay_0
   widget=SizeBox_0
   properties {bOverride_MaxDesiredHeight=true, bOverride_MinDesiredHeight=true, maxDesiredHeight=488, minDesiredHeight=370}
   slot {class=OverlaySlot, properties {horizontalAlignment=HAlign_Fill, verticalAlignment=VAlign_Fill, padding=<ltrb 8 8 8 8>}}
  }
  TileView_ModList {
   widgetClassPath=/Script/FortniteUI.FortTileView
   bIsVariable=true
   parent=SizeBox_0
   widget=TileView_ModList
   properties {
    entryHeight=118
    entryWidgetClass=&/WBP_ArmTale_ModListEntry.WBP_ArmTale_ModListEntry_C
    entryWidth=118
    horizontalEntrySpacing=16
    numDesignerPreviewEntries=20
    scrollIntoViewAlignment=TopOrLeft
    scrollbarDisabledVisibility=Hidden
    tileAlignment=LeftAligned
    verticalEntrySpacing=16
    scrollBarStyle {
     thickness=8
     draggedThumbImage {
      bIsDynamicallyLoaded=false
      drawAs=Box
      imageType=FullColor
      mirroring=NoMirror
      resourceName=../../../Engine/Content/Slate/Common/Scrollbar_Thumb.png
      resourceObject=null
      tiling=NoTile
      imageSize=<0 0>
      margin=<ltrb 0.25 0.25 0.25 0.25>
      outlineSettings {
       bUseBrushTransparency=false
       roundingType=HalfHeightRadius
       width=0
       color {colorUseRule=UseColor_Specified, specifiedColor=<rgba 0 0 0 0>}
       cornerRadii=<0 0 0 0>
      }
      tintColor {colorUseRule=UseColor_Specified, specifiedColor=<rgba 1 1 1 1>}
      uVRegion {bIsValid=false, max=<0 0>, min=<0 0>}
     }
     horizontalBackgroundImage {
      bIsDynamicallyLoaded=false
      drawAs=NoDrawType
      imageType=NoImage
      mirroring=NoMirror
      resourceName=None
      resourceObject=null
      tiling=NoTile
      imageSize=<0 0>
      margin=<ltrb 0 0 0 0>
      outlineSettings {
       bUseBrushTransparency=false
       roundingType=HalfHeightRadius
       width=0
       color {colorUseRule=UseColor_Specified, specifiedColor=<rgba 0 0 0 0>}
       cornerRadii=<0 0 0 0>
      }
      tintColor {colorUseRule=UseColor_Specified, specifiedColor=<rgba 1 1 1 1>}
      uVRegion {bIsValid=false, max=<0 0>, min=<0 0>}
     }
     horizontalBottomSlotImage {
      bIsDynamicallyLoaded=false
      drawAs=Image
      imageType=FullColor
      mirroring=NoMirror
      resourceName=../../../Engine/Content/Slate/Common/Scrollbar_Background_Horizontal.png
      resourceObject=null
      tiling=NoTile
      imageSize=<8 8>
      margin=<ltrb 0 0 0 0>
      outlineSettings {
       bUseBrushTransparency=false
       roundingType=HalfHeightRadius
       width=0
       color {colorUseRule=UseColor_Specified, specifiedColor=<rgba 0 0 0 0>}
       cornerRadii=<0 0 0 0>
      }
      tintColor {colorUseRule=UseColor_Specified, specifiedColor=<rgba 1 1 1 1>}
      uVRegion {bIsValid=false, max=<0 0>, min=<0 0>}
     }
     horizontalTopSlotImage {
      bIsDynamicallyLoaded=false
      drawAs=Image
      imageType=FullColor
      mirroring=NoMirror
      resourceName=../../../Engine/Content/Slate/Common/Scrollbar_Background_Horizontal.png
      resourceObject=null
      tiling=NoTile
      imageSize=<8 8>
      margin=<ltrb 0 0 0 0>
      outlineSettings {
       bUseBrushTransparency=false
       roundingType=HalfHeightRadius
       width=0
       color {colorUseRule=UseColor_Specified, specifiedColor=<rgba 0 0 0 0>}
       cornerRadii=<0 0 0 0>
      }
      tintColor {colorUseRule=UseColor_Specified, specifiedColor=<rgba 1 1 1 1>}
      uVRegion {bIsValid=false, max=<0 0>, min=<0 0>}
     }
     hoveredThumbImage {
      bIsDynamicallyLoaded=false
      drawAs=Box
      imageType=FullColor
      mirroring=NoMirror
      resourceName=../../../Engine/Content/Slate/Common/Scrollbar_Thumb.png
      resourceObject=null
      tiling=NoTile
      imageSize=<0 0>
      margin=<ltrb 0.25 0.25 0.25 0.25>
      outlineSettings {
       bUseBrushTransparency=false
       roundingType=HalfHeightRadius
       width=0
       color {colorUseRule=UseColor_Specified, specifiedColor=<rgba 0 0 0 0>}
       cornerRadii=<0 0 0 0>
      }
      tintColor {colorUseRule=UseColor_Specified, specifiedColor=<rgba 1 1 1 1>}
      uVRegion {bIsValid=false, max=<0 0>, min=<0 0>}
     }
     normalThumbImage {
      bIsDynamicallyLoaded=false
      drawAs=Box
      imageType=FullColor
      mirroring=NoMirror
      resourceName=../../../Engine/Content/Slate/Common/Scrollbar_Thumb.png
      resourceObject=null
      tiling=NoTile
      imageSize=<0 0>
      margin=<ltrb 0.25 0.25 0.25 0.25>
      outlineSettings {
       bUseBrushTransparency=false
       roundingType=HalfHeightRadius
       width=0
       color {colorUseRule=UseColor_Specified, specifiedColor=<rgba 0 0 0 0>}
       cornerRadii=<0 0 0 0>
      }
      tintColor {colorUseRule=UseColor_Specified, specifiedColor=<rgba 1 1 1 1>}
      uVRegion {bIsValid=false, max=<0 0>, min=<0 0>}
     }
     verticalBackgroundImage {
      bIsDynamicallyLoaded=false
      drawAs=Image
      imageType=NoImage
      mirroring=NoMirror
      resourceName=None
      resourceObject=null
      tiling=NoTile
      imageSize=<0 0>
      margin=<ltrb 0 0 0 0>
      outlineSettings {
       bUseBrushTransparency=false
       roundingType=HalfHeightRadius
       width=0
       color {colorUseRule=UseColor_Specified, specifiedColor=<rgba 0 0 0 0>}
       cornerRadii=<0 0 0 0>
      }
      tintColor {colorUseRule=UseColor_Specified, specifiedColor=<rgba 0 0 0 0.6>}
      uVRegion {bIsValid=false, max=<0 0>, min=<0 0>}
     }
     verticalBottomSlotImage {
      bIsDynamicallyLoaded=false
      drawAs=NoDrawType
      imageType=FullColor
      mirroring=NoMirror
      resourceName=../../../Engine/Content/Slate/Common/Scrollbar_Background_Vertical.png
      resourceObject=null
      tiling=NoTile
      imageSize=<0 0>
      margin=<ltrb 0 0 0 0>
      outlineSettings {
       bUseBrushTransparency=false
       roundingType=HalfHeightRadius
       width=0
       color {colorUseRule=UseColor_Specified, specifiedColor=<rgba 0 0 0 0>}
       cornerRadii=<0 0 0 0>
      }
      tintColor {colorUseRule=UseColor_Specified, specifiedColor=<rgba 1 1 1 1>}
      uVRegion {bIsValid=false, max=<0 0>, min=<0 0>}
     }
     verticalTopSlotImage {
      bIsDynamicallyLoaded=false
      drawAs=NoDrawType
      imageType=FullColor
      mirroring=NoMirror
      resourceName=../../../Engine/Content/Slate/Common/Scrollbar_Background_Vertical.png
      resourceObject=null
      tiling=NoTile
      imageSize=<0 0>
      margin=<ltrb 0 0 0 0>
      outlineSettings {
       bUseBrushTransparency=false
       roundingType=HalfHeightRadius
       width=0
       color {colorUseRule=UseColor_Specified, specifiedColor=<rgba 0 0 0 0>}
       cornerRadii=<0 0 0 0>
      }
      tintColor {colorUseRule=UseColor_Specified, specifiedColor=<rgba 1 1 1 1>}
      uVRegion {bIsValid=false, max=<0 0>, min=<0 0>}
     }
    }
   }
   slot {class=SizeBoxSlot}
  }
 ]
}
