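# WBP_ArmTale_RarityLabel is a UI label showing an item rarity's name in uppercase over a sheared, material-based background.
# SetRarity gets the Fort rarity data for an EFortRarity, caches it, and returns it.
# It sets the text to the rarity name, tinting the text with Color1 and the background with Color4.
# PreConstruct applies PreviewRarity only at design time, so at runtime the label relies on SetRarity being called.
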WidgetBlueprint /ArmTaleFrontendUI/Widgets/WBP_ArmTale_RarityLabel
graphs [
 EventGraph {nodes [
   n1 = Event PreConstruct()
   > Branch(Condition=@n1.IsDesignTime)
   > SetRarity!(execute, self, Rarity:EFortRarity Enum=Get PreviewRarity() -> (PreviewRarity:EFortRarity Enum=Common)) -> (then, RarityData:Fort Rarity Item Data Structure)
  ]}
 SetRarity {nodes [
   n1 = SetRarity!() -> (then, Rarity:EFortRarity Enum=Common) {is=title K2Node_FunctionEntry}
   > n2 = GetRarityData!(execute, Rarity:EFortRarity Enum=@n1) -> (then, ReturnValue:Fort Rarity Item Data Structure)
   > n3 = Set CacheRarityData!(execute, CacheRarityData:Fort Rarity Item Data Structure=@n2) -> (
    then
    Output_Get_Name:Text
    Output_Get_Color1:Linear Color Structure
    Output_Get_Color2:Linear Color Structure
    Output_Get_Color3:Linear Color Structure
    Output_Get_Color4:Linear Color Structure
    Output_Get_Color5:Linear Color Structure
    Output_Get_Radius:float=~
    Output_Get_Falloff:float=~
    Output_Get_Brightness:float=~
    Output_Get_Roughness:float=~
    Output_Get_Glow:float=~
   )
   n4 = Get TextBlock_RarityName() -> (TextBlock_RarityName:Common Text Object Reference)
   > "SetText (Text)"(self=@n4, InText=@n3)
   > SetColorAndOpacity(self=@n4, InColorAndOpacity=Make Slate Color(SpecifiedColor=@n3.Output_Get_Color1))
   > SetColorAndOpacity(
    self:Image Object Reference=Get Image_Background() -> (Image_Background:Image Object Reference)
    InColorAndOpacity:Linear Color Structure=@n3.Output_Get_Color4
   )
   > Return Node!(execute, RarityData:Fort Rarity Item Data Structure=Get CacheRarityData() -> (CacheRarityData:Fort Rarity Item Data Structure))
  ]}
]
variables [
 CacheRarityData {
  type=FortRarityItemData
  default {
   name=""
   brightness=0
   falloff=0
   glow=0
   radius=0
   roughness=0
   color1=<rgba 0 0 0 1>
   color2=<rgba 0 0 0 1>
   color3=<rgba 0 0 0 1>
   color4=<rgba 0 0 0 1>
   color5=<rgba 0 0 0 1>
  }
 }
 PreviewRarity {type=EFortRarity, default=Rare}
]
widgets {
 info {inheritedWidgetCount=0, parentClass=&UserWidget, rootWidgetClass=&Overlay, widgetCount=3}
 widgets [
  Overlay_17 {widgetClassPath=&Overlay, widget=Overlay_17}
  Image_Background {
   widgetClassPath=&Image
   bIsVariable=true
   parent=Overlay_17
   widget=Image_Background
   properties {
    brush {
     bIsDynamicallyLoaded=false
     drawAs=Image
     imageType=FullColor
     mirroring=NoMirror
     resourceName=None
     resourceObject=/ArmTaleFrontendUI/Assets/Materials/MI_UI_RarityLabelBG
     tiling=NoTile
     imageSize=<16 16>
     margin=<ltrb 0 0 0 0>
     outlineSettings {
      bUseBrushTransparency=false
      roundingType=HalfHeightRadius
      width=0
      color {colorUseRule=UseColor_Specified, specifiedColor=<rgba 0 0 0 0>}
      cornerRadii=<0 0 0 0>
     }
     tintColor {colorUseRule=UseColor_Specified, specifiedColor=<rgba 1 1 1 1>}
     uVRegion {bIsValid=false, max=<0 0>, min=<0 0>}
    }
    colorAndOpacity=<rgba 0 0 0 1>
    renderTransform {angle=0, scale=<1 1>, shear=<-16 0>, translation=<0 0>}
   }
   slot {class=OverlaySlot, properties {horizontalAlignment=HAlign_Fill, verticalAlignment=VAlign_Fill}}
  }
  TextBlock_RarityName {
   widgetClassPath=&CommonTextBlock
   bIsVariable=true
   parent=Overlay_17
   widget=TextBlock_RarityName
   properties {
    text="{Rarity}"
    textTransformPolicy=ToUpper
    font {
     bForceMonospaced=false
     bMaterialIsStencil=false
     fontMaterial=null
     fontObject=&BurbankSmall
     letterSpacing=0
     monospacedWidth=1
     size=13.5
     skewAmount=0
     typefaceFontName=Black
     outlineSettings {
      bApplyOutlineToDropShadows=false
      bMiteredCorners=false
      bSeparateFillAlpha=false
      outlineMaterial=null
      outlineSize=0
      outlineColor=<rgba 0 0 0 1>
     }
    }
    shadowOffset=<0 0>
   }
   slot {class=OverlaySlot, properties {horizontalAlignment=HAlign_Center, verticalAlignment=VAlign_Center, padding=<ltrb 8 2 8 2>}}
  }
 ]
}
