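# Common UI button widget that displays an input binding for the ArmTale frontend, built on CommonButtonBase with a CommonActionWidget as its root.
# SetInputAction takes a data table row handle and passes it to InputActionWidget for display.
# It then sets the same row as the button's triggering input action, so the shown prompt and the triggering input match.
# The event graph is empty.
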WidgetBlueprint /ArmTaleFrontendUI/Widgets/WBP_ArmTale_InputBinding
graphs [
 EventGraph {links [], nodes []}
 SetInputAction {nodes [
   n1 = SetInputAction!() -> (then, InputAction:Data Table Row Handle Structure) {is=title K2Node_FunctionEntry}
   > SetInputAction!(
    execute
    self:Common Action Widget Object Reference=Get InputActionWidget() -> (InputActionWidget:Common Action Widget Object Reference)
    InputActionRow:Data Table Row Handle Structure=@n1
   ) -> (then) {is=title K2Node_CallFunction}
   > SetTriggeringInputAction!(execute, self:Common Button Base Object Reference, InputActionRow:"Data Table Row Handle Structure (by ref)"=@n1) -> (then)
  ]}
]
variables []
widgets {
 info {inheritedWidgetCount=1, parentClass=&CommonButtonBase, rootWidgetClass=&CommonActionWidget, widgetCount=1}
 widgets [InputActionWidget {widgetClassPath=&CommonActionWidget, bInherited=true, bIsVariable=true, widget=InputActionWidget}]
}
