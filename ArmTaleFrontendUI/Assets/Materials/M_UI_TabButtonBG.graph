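# Translucent UI material for a tab button background, drawing a tapered quadrilateral from a procedural distance field.
# It measures distance to the horizontal and vertical edges from normalized UVs, correcting X by the pixel-size aspect ratio.
# DistortX and DistortY offset each distance by the opposite axis, tapering the shape.
# The smaller distance feeds MF_UI_AntiAliasing, with Padding as threshold, for smooth edges without textures.
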Material /ArmTaleFrontendUI/Assets/Materials/M_UI_TabButtonBG
graph {
 properties {
  bAutomaticallySetUsageInEditor=false
  bUsedWithStaticMesh=false
  blendMode=BLEND_Translucent
  materialDomain=MD_UI
  usedShadingModels=Default Lit
 }
 nodes [
  Constant() {r=1}
  n1 = MaterialFunctionCall() {materialFunction=&GetUserInterfaceUV}
  n2 = Multiply(A=Divide(A=ComponentMask(None=@n1.Pixel Size) {r=true}, B=ComponentMask(None=@n1.Pixel Size) {g=true})) {constB=2}
  n3 = ComponentMask(None=@n1.Normalized UV) {r=true}
  n4 = Min(A=Multiply(A=@n2, B=@n3), B=Multiply(A=@n2, B=OneMinus(None=@n3)))
  n5 = ConstantBiasScale(None=ComponentMask(None=@n1.Normalized UV) {g=true}) {bias=-0.5, scale=2}
  n6 = Multiply(A=@n5, B=ScalarParameter() {defaultValue=-0.2, parameterName=DistortX}) {constB=0.1}
  n7 = Multiply(A=ConstantBiasScale(None=@n3) {bias=-0.5, scale=2}, B=ScalarParameter() {defaultValue=0.1, parameterName=DistortY}) {constB=0.1}
  n8 = OneMinus(None=Abs(None=@n5))
  MaterialFunctionCall(DistanceField=Min(A=Add(A=@n4, B=@n6), B=Add(A=@n7, B=@n8)), Threshold=ScalarParameter() {defaultValue=0.25, parameterName=Padding}) {materialFunction=&MF_UI_AntiAliasing}
 ]
}
