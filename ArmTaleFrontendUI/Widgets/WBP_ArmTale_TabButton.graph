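# Tab button widget for the ArmTale frontend UI, built on CommonButtonLegacy with an uppercase label over a material-driven background.
# PreConstruct copies ButtonText into the label.
# Hover events play AnimHovered forward and in reverse unless the button is selected, while selection events do the same with AnimSelected.
# Both animations fade in the background and animate its material padding and distortion; selection reaches full opacity and turns the label black.
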WidgetBlueprint /ArmTaleFrontendUI/Widgets/WBP_ArmTale_TabButton
animations [
 AnimHovered {
  binding_count=2
  length_seconds=0.15
  object=&self:AnimHovered
  sequence {bindings [
    Image_Background {
     id=#1
     possessed_class=Image
     tracks [
      {
       type=MovieSceneWidgetMaterialTrack
       sections [{
         type=MovieSceneParameterSection
         end_frame=3
         start_frame=0
         channels [
          Padding {type=MovieSceneScriptingFloatChannel, keys [[0,0.5],[3,0.25]]}
          DistortX {type=MovieSceneScriptingFloatChannel, keys [[0,-0.05],[3,-0.2]]}
          DistortY {type=MovieSceneScriptingFloatChannel, keys [[0,0],[3,0.1]]}
         ]
        }]
      }
      {
       type=MovieSceneFloatTrack
       display_name=Render Opacity
       sections [{type=MovieSceneFloatSection, end_frame=3, start_frame=0, channels [None {type=MovieSceneScriptingFloatChannel, keys [[0,0],[3,0.25]]}]}]
      }
     ]
    }
    TextBlock_Label {
     id=#2
     possessed_class=CommonTextBlock
     tracks [{
       type=MovieSceneColorTrack
       display_name=Color and Opacity
       sections [{
         type=MovieSceneColorSection
         end_frame=3
         start_frame=0
         channels [
          Color.R {type=MovieSceneScriptingFloatChannel, const [1,1]}
          Color.G {type=MovieSceneScriptingFloatChannel, const [1,1]}
          Color.B {type=MovieSceneScriptingFloatChannel, const [1,1]}
          Color.A {type=MovieSceneScriptingFloatChannel, const [1,1]}
         ]
        }]
      }]
    }
   ]}
 }
 AnimSelected {
  binding_count=2
  length_seconds=0.15
  object=&self:AnimSelected
  sequence {bindings [
    Image_Background {
     id=#1
     possessed_class=Image
     tracks [
      {
       type=MovieSceneWidgetMaterialTrack
       sections [{
         type=MovieSceneParameterSection
         end_frame=3
         start_frame=0
         channels [
          Padding {type=MovieSceneScriptingFloatChannel, keys [[0,0.5],[3,0.25]]}
          DistortX {type=MovieSceneScriptingFloatChannel, keys [[0,-0.05],[3,-0.2]]}
          DistortY {type=MovieSceneScriptingFloatChannel, keys [[0,0],[3,0.1]]}
         ]
        }]
      }
      {
       type=MovieSceneFloatTrack
       display_name=Render Opacity
       sections [{type=MovieSceneFloatSection, end_frame=3, start_frame=0, channels [None {type=MovieSceneScriptingFloatChannel, keys [[0,0],[3,1]]}]}]
      }
     ]
    }
    TextBlock_Label {
     id=#2
     possessed_class=CommonTextBlock
     tracks [{
       type=MovieSceneColorTrack
       display_name=Color and Opacity
       sections [{
         type=MovieSceneColorSection
         end_frame=3
         start_frame=0
         channels [
          Color.R {type=MovieSceneScriptingFloatChannel, keys [[0,1],[3,0]]}
          Color.G {type=MovieSceneScriptingFloatChannel, keys [[0,1],[3,0]]}
          Color.B {type=MovieSceneScriptingFloatChannel, keys [[0,1],[3,0]]}
          Color.A {type=MovieSceneScriptingFloatChannel, const [1,1]}
         ]
        }]
      }]
    }
   ]}
 }
]
graphs [EventGraph {nodes [
   Event On Hovered()
   n1 = GetSelected!(self:Common Button Base Object Reference) -> (ReturnValue:bool)
   > n2 = Branch(Condition=@n1)
   n3 = Get AnimHovered() -> (AnimHovered:Widget Animation Object Reference)
   PlayAnimationForward(execute=@n2.else, InAnimation=@n3)
   Event PreConstruct()
   > "SetText (Text)"(self=Get TextBlock_Label() -> (TextBlock_Label:Common Text Object Reference), InText=Get ButtonText() -> (ButtonText:Text))
   Event On Unhovered()
   > n4 = Branch(Condition=@n1)
   PlayAnimationReverse(execute=@n4.else, InAnimation=@n3)
   Event On Selected()
   n5 = Get AnimSelected() -> (AnimSelected:Widget Animation Object Reference)
   > PlayAnimationForward(InAnimation=@n5)
   Event On Deselected()
   > PlayAnimationReverse(InAnimation=@n5)
  ]}]
variables [ButtonText {type=string, default="{Button Text}"}]
widgets {
 info {inheritedWidgetCount=1, parentClass=&CommonButtonLegacy, rootWidgetClass=&Overlay, widgetCount=4}
 widgets [
  Overlay_17 {widgetClassPath=&Overlay, widget=Overlay_17}
  Image_Background {
   widgetClassPath=&Image
   bIsVariable=true
   parent=Overlay_17
   widget=Image_Background
   properties {
    renderOpacity=0
    brush {
     bIsDynamicallyLoaded=false
     drawAs=Image
     imageType=FullColor
     mirroring=NoMirror
     resourceName=None
     resourceObject=/ArmTaleFrontendUI/Assets/Materials/M_UI_TabButtonBG
     tiling=NoTile
     imageSize=<16 16>
     margin=<ltrb 0 0 0 0>
     outlineSettings {
      bUseBrushTransparency=false
      roundingType=HalfHeightRadius
      width=0
      color {colorUseRule=UseColor_Specified, specifiedColor=<rgba 0 0 0 0>}
      cornerRadii=<0 0 0 0>
     }
     tintColor {colorUseRule=UseColor_Specified, specifiedColor=<rgba 1 1 1 1>}
     uVRegion {bIsValid=false, max=<0 0>, min=<0 0>}
    }
   }
   slot {class=OverlaySlot, properties {horizontalAlignment=HAlign_Fill, verticalAlignment=VAlign_Fill}}
  }
  TextBlock_Label {
   widgetClassPath=&CommonTextBlock
   bIsVariable=true
   parent=Overlay_17
   widget=TextBlock_Label
   properties {
    text="{Button Text}"
    textTransformPolicy=ToUpper
    font {
     bForceMonospaced=false
     bMaterialIsStencil=false
     fontMaterial=null
     fontObject=&HeadingNow
     letterSpacing=0
     monospacedWidth=1
     size=13.5
     skewAmount=0
     typefaceFontName=76 Bold
     outlineSettings {
      bApplyOutlineToDropShadows=false
      bMiteredCorners=false
      bSeparateFillAlpha=false
      outlineMaterial=null
      outlineSize=0
      outlineColor=<rgba 0 0 0 1>
     }
    }
    shadowOffset=<0 0>
   }
   slot {class=OverlaySlot, properties {horizontalAlignment=HAlign_Center, verticalAlignment=VAlign_Center, padding=<ltrb 18 12 18 12>}}
  }
  InputActionWidget {widgetClassPath=&CommonActionWidget, bInherited=true, widget=null}
 ]
}
