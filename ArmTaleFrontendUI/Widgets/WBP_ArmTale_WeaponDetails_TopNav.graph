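# Top navigation bar for the ArmTale weapon details screen: Showcase tab, dynamic mod slots and a disabled, dimmed Customize tab.
# HandleModSlotsChanged rebuilds the slot widgets and a selection-required CommonUI button group in that tab order.
# HandleEquippedModsChanged pushes each equipped mod's placeholder (FAKE) viewmodel into the slot at the same index.
# Prev/Next input bindings on MainTabLeft/MainTabRight step through the group with wrapping.
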WidgetBlueprint /ArmTaleFrontendUI/Widgets/WBP_ArmTale_WeaponDetails_TopNav
graphs [
 EventGraph {nodes [
   "On Clicked (PrevTabBinding)"!() -> (OutputDelegate, then, Button:Common Button Base Object Reference) {is=title K2Node_ComponentBoundEvent}
   n1 = Get ButtonGroup() -> (ButtonGroup:Common Button Group Base Object Reference)
   > SelectPreviousButton!(execute, self:Common Button Group Base Object Reference=@n1, bAllowWrap:bool=true, bAllowSound:bool=true) -> (then)
   "On Clicked (NextTabBinding)"!() -> (OutputDelegate, then, Button:Common Button Base Object Reference) {is=title K2Node_ComponentBoundEvent}
   > SelectNextButton!(execute, self:Common Button Group Base Object Reference=@n1, bAllowWrap:bool=true, bAllowSound:bool=true) -> (then)
   Event OnInitialized()
   > n2 = CreateButtonGroup!(execute, bInSelectionRequired:bool=true) -> (then, ReturnValue:Common Button Group Legacy Object Reference)
   > Set ButtonGroup(ButtonGroup:Common Button Group Base Object Reference=@n2) -> (Output_Get:Common Button Group Base Object Reference)
  ]}
 HandleEquippedModsChanged {nodes [
   HandleEquippedModsChanged!() -> (then, EquippedMods:Array of VM Arm Tale Mod Entry FAKE Object References) {is=title K2Node_FunctionEntry}
   > n1 = For Each Loop(Array:Array of Widget Object References=GetAllChildren(self=Get HBox_ModSlots() -> (HBox_ModSlots:Horizontal Box Object Reference))) -> (Array Element:Widget Object Reference)
   n2 = Get EquippedMods() -> (EquippedMods:Array of VM Arm Tale Mod Entry FAKE Object References)
   > Branch(Condition=Is Valid Index!(TargetArray:Array of VM Arm Tale Mod Entry FAKE Object References=@n2, IndexToTest:int=@n1.Array Index) -> (ReturnValue:bool))
   > n3 = Cast To WBP_ArmTale_ModSlot!(execute, Object=@n1) -> (then, CastFailed, AsWBP Arm Tale Mod Slot:WBP Arm Tale Mod Slot Object Reference) {is=title K2Node_DynamicCast}
   > SetVM_ArmTale_ModEntry_FAKE!(
    execute
    self:WBP Arm Tale Mod Slot Object Reference=@n3
    Viewmodel:VM Arm Tale Mod Entry FAKE Object Reference="Get (a copy)"(Array:Array of VM Arm Tale Mod Entry FAKE Object References=@n2, Dimension 1=@n1.Array Index) -> (Output:VM Arm Tale Mod Entry FAKE Object Reference)
   ) -> (then)
  ]}
 HandleModSlotsChanged {nodes [
   HandleModSlotsChanged!() -> (then, ModSlots:Array of E_ArmTale_ModTypes_FAKE Enums=NewEnumerator3) {is=title K2Node_FunctionEntry}
   > ClearChildren(self=Get HBox_ModSlots())
   > n1 = CreateButtonGroup(bInSelectionRequired=true)
   > n2 = Set ButtonGroup(ButtonGroup=@n1)
   > AddWidget!(
    execute
    self:Common Widget Group Base Object Reference=@n2
    InWidget:Widget Object Reference=Get Button_Showcase() -> (Button_Showcase:WBP Arm Tale Tab Button Object Reference)
   ) -> (then)
   > n3 = For Each Loop(Array:Array of E_ArmTale_ModTypes_FAKE Enums=Get ModSlots() -> (ModSlots:Array of E_ArmTale_ModTypes_FAKE Enums=NewEnumerator3)) -> (Array Element:E_ArmTale_ModTypes_FAKE Enum)
   AddWidget(
    execute=@n3.Completed
    self=Get ButtonGroup()
    InWidget=Get Button_Customize() -> (Button_Customize:WBP Arm Tale Tab Button Object Reference)
   )
   Set NewModSlotType(execute=@n3.LoopBody, NewModSlotType:E_ArmTale_ModTypes_FAKE Enum=@n3) -> (Output_Get:E_ArmTale_ModTypes_FAKE Enum=NewEnumerator3)
   > Set NewModSlotIndex(NewModSlotIndex:int=@n3.Array Index) -> (Output_Get:int=0)
   > n4 = Create Widget!(
    execute:Exec
    Class:User Widget Class Reference=&/WBP_ArmTale_ModSlot.WBP_ArmTale_ModSlot_C
    OwningPlayer:Player Controller Object Reference
    VM_ArmTale_ModEntry_FAKE:VM Arm Tale Mod Entry FAKE Object Reference
    Style:Common Button Style Class Reference=/FortUILibrary/Foundation/Buttons/ButtonStyles/ButtonStyle-Empty.ButtonStyle-Empty_C
    bApplyAlphaOnDisable:bool=True
    bLocked:bool=False
    bSelectable:bool=True
    bShouldSelectUponReceivingFocus:bool=False
    bInteractableWhenSelected:bool=False
    bToggleable:bool=False
    bTriggerClickedAfterSelection:bool=False
    bDisplayInputActionWhenNotInteractable:bool=True
    bHideInputActionWithKeyboard:bool=False
    bShouldUseFallbackDefaultInputAction:bool=True
    bRequiresHold:bool=False
    HoldData:Common UIHold Data Class Reference
    ClickMethod:EButtonClickMethod Enum=DownAndUp
    InputPriority:int=0
    TriggeringInputAction:Data Table Row Handle Structure
   ) -> (then:Exec, ReturnValue:WBP Arm Tale Mod Slot Object Reference)
   > n5 = Set NewModSlot(NewModSlot:WBP Arm Tale Mod Slot Object Reference=@n4) -> (Output_Get:WBP Arm Tale Mod Slot Object Reference)
   > AddWidget(self=Get ButtonGroup(), InWidget=@n5)
   > n6 = AddChildToHorizontalBox!(execute, self:Horizontal Box Object Reference=Get HBox_ModSlots(), Content:Widget Object Reference=@n5) -> (then, ReturnValue:Horizontal Box Slot Object Reference)
   > SetPadding!(
    execute
    self:Horizontal Box Slot Object Reference=@n6
    InPadding_Left:float=12
    InPadding_Top:float=0
    InPadding_Right:float=12
    InPadding_Bottom:float=0
   ) -> (then)
   > Set ModSlotIndex!(
    execute
    ModSlotIndex:int=Get NewModSlotIndex() -> (NewModSlotIndex:int=0)
    self:WBP Arm Tale Mod Slot Object Reference=Get NewModSlot() -> (NewModSlot:WBP Arm Tale Mod Slot Object Reference)
   ) -> (then, Output_Get:int=0)
  ]}
]
variables [ButtonGroup {type=CommonButtonGroupBase, default=null}]
widgets {
 info {inheritedWidgetCount=0, parentClass=&CommonUserWidget, rootWidgetClass=&HorizontalBox, widgetCount=6}
 widgets [
  HorizontalBox {widgetClassPath=&HorizontalBox, widget=HorizontalBox}
  PrevTabBinding {
   widgetClassPath=&/WBP_ArmTale_InputBinding
   bIsVariable=true
   parent=HorizontalBox
   widget=PrevTabBinding
   properties {
    holdData=&CommonUI_HoldData_Base_C
    triggeringInputAction {dataTable=&InputActionDataTable, rowName=MainTabLeft}
   }
   slot {class=HorizontalBoxSlot, properties {horizontalAlignment=HAlign_Center, verticalAlignment=VAlign_Center}}
  }
  Button_Showcase {
   widgetClassPath=&/WBP_ArmTale_TabButton
   bIsVariable=true
   parent=HorizontalBox
   widget=Button_Showcase
   properties {
    bIsFocusable=false
    bSelectable=true
    buttonText=Showcase
    holdData=&CommonUI_HoldData_Base_C
    onButtonBaseClicked=()
    onButtonBaseDoubleClicked=()
    onButtonBaseHovered=()
    onButtonBaseUnhovered=()
    onSelectedChangedBase=()
   }
   slot {class=HorizontalBoxSlot, properties {verticalAlignment=VAlign_Center, padding=<ltrb 16 0 16 0>}}
  }
  HBox_ModSlots {widgetClassPath=&HorizontalBox, bIsVariable=true, parent=HorizontalBox, widget=HBox_ModSlots, slot {class=HorizontalBoxSlot}}
  Button_Customize {
   widgetClassPath=&WBP_ArmTale_TabButton
   bIsVariable=true
   parent=HorizontalBox
   widget=Button_Customize
   properties {
    bIsEnabled=false
    bIsFocusable=false
    bSelectable=true
    buttonText=Customize
    holdData=&CommonUI_HoldData_Base_C
    onButtonBaseClicked=()
    onButtonBaseDoubleClicked=()
    onButtonBaseHovered=()
    onButtonBaseUnhovered=()
    onSelectedChangedBase=()
    renderOpacity=0.2
   }
   slot {class=HorizontalBoxSlot, properties {verticalAlignment=VAlign_Center, padding=<ltrb 16 0 16 0>}}
  }
  NextTabBinding {
   widgetClassPath=&WBP_ArmTale_InputBinding
   bIsVariable=true
   parent=HorizontalBox
   widget=NextTabBinding
   properties {
    holdData=&CommonUI_HoldData_Base_C
    triggeringInputAction {dataTable=&InputActionDataTable, rowName=MainTabRight}
   }
   slot {class=HorizontalBoxSlot, properties {horizontalAlignment=HAlign_Center, verticalAlignment=VAlign_Center}}
  }
 ]
}
