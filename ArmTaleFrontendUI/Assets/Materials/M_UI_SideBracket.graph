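# M_UI_SideBracket is a translucent UI material that procedurally builds a side bracket shape from signed distance fields, without textures.
# 9-slice UVs are rotated a quarter turn, mirrored, offset and scaled to feed a three-sided polygon SDF.
# The same base UVs, shifted and swizzled, feed an isosceles trapezoid SDF.
# An SDF subtraction combines both shapes, then anti-aliasing smooths the edge, with parameters for polygon radius and trapezoid dimensions.
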Material /ArmTaleFrontendUI/Assets/Materials/M_UI_SideBracket
graph {
 properties {
  bAutomaticallySetUsageInEditor=false
  bUsedWithStaticMesh=false
  blendMode=BLEND_Translucent
  materialDomain=MD_UI
  usedShadingModels=Default Lit
 }
 nodes [
  Constant() {r=1}
  n1 = MaterialFunctionCall!(UVs) {materialFunction=/FortUILibrary/MaterialFunctions/Basics/MF_UI_9SliceUVs}
  n2 = MaterialFunctionCall(UVs=@n1.UVs, "Rotation Angle (0-1)"=Constant() {r=-0.25}) {materialFunction=&CustomRotator}
  n3 = MaterialFunctionCall(
   UVs=Add(
    A=MaterialFunctionCall!(UV=@n2.Rotated Values, Mirror Y, Swizzle UV) {materialFunction=/FortUILibrary/MaterialFunctions/Basics/MF_UI_Symmetry_2x1_New}.Symmetry
    B=Constant2Vector() {g=0.1, r=-0.55}
   )
   Texture Scale=Constant2Vector() {g=2, r=2.5}
  ) {materialFunction=&ScaleUVsByCenter}
  n4 = ScalarParameter() {defaultValue=0.9, parameterName=Radius}
  n5 = Constant() {r=3}
  n6 = MaterialFunctionCall(XY=Subtract(A=@n1.UVs, B=Constant2Vector() {r=0.435})) {materialFunction=&Swizzle}
  n7 = ScalarParameter() {defaultValue=0.45, parameterName=TopWidth}
  n8 = ScalarParameter() {defaultValue=0.35, parameterName=BottomWidth}
  n9 = ScalarParameter() {defaultValue=0.95, parameterName=Height}
  MaterialFunctionCall(DistanceField=MaterialFunctionCall!(
    DistanceFieldA=MaterialFunctionCall!(UVs=@n3.UVs, Radius=@n4, Sides=@n5, InnerAngle) {materialFunction=/FortUILibrary/MaterialFunctions/Shapes/SDFs/MF_SDF_RegularPolygon}.SDF
    DistanceFieldB=MaterialFunctionCall!(UVs=@n6.YX, TopWidth=@n7, BottomWidth=@n8, Height=@n9) {materialFunction=/FortUILibrary/MaterialFunctions/Shapes/SDFs/MF_SDF_IsoscelesTrapezoid}.SDF
    BlendAmount
   ) {materialFunction=/FortUILibrary/MaterialFunctions/Shapes/Modifiers/MF_UI_SDF_Subtraction}.Exact) {materialFunction=&MF_UI_AntiAliasing}
 ]
}
