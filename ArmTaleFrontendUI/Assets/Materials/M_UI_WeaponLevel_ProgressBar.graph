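# Translucent UI material for a weapon level progress bar that previews a gain or loss against current progress.
# Progress and Progress plus PreviewValue, each divided by MaxValue, become antialiased thresholds along horizontal normalized UV.
# Their overlap is white, while the difference is green for gains and red for losses, swappable through FlipColors.
# The unfilled area is black, with BackgroundOpacity blending to full opacity across the filled area.
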Material /ArmTaleFrontendUI/Assets/Materials/M_UI_WeaponLevel_ProgressBar
graph {
 properties {
  bAutomaticallySetUsageInEditor=false
  bUsedWithStaticMesh=false
  blendMode=BLEND_Translucent
  materialDomain=MD_UI
  usedShadingModels=Default Lit
 }
 nodes [
  n1 = ScalarParameter() {defaultValue=1, parameterName=MaxValue}
  n2 = ScalarParameter() {defaultValue=0.4, parameterName=Progress}
  n3 = Divide(A=@n2, B=@n1)
  n4 = ComponentMask(None=MaterialFunctionCall() {materialFunction=&GetUserInterfaceUV}.Normalized UV) {r=true}
  n5 = MaterialFunctionCall(DistanceField=@n4, Threshold=@n3) {materialFunction=&MF_UI_AntiAliasing}
  n6 = Divide(A=Add(A=@n2, B=ScalarParameter() {parameterName=PreviewValue}), B=@n1)
  n7 = MaterialFunctionCall(DistanceField=@n4, Threshold=@n6) {materialFunction=&MF_UI_AntiAliasing}
  n8 = Max(A=@n5.Inner, B=@n7.Inner)
  LinearInterpolate(A=ScalarParameter() {defaultValue=0.6, parameterName=BackgroundOpacity}, Alpha=@n8) {constA=0.4}
  n9 = Constant3Vector() {constant=<rgba 0.0331 1 0.2159 1>}
  n10 = Constant3Vector() {constant=<rgba 1 0.013 0.0931 1>}
  n11 = MaterialFunctionCall!(A=@n3, B=@n6) {materialFunction=/FortUILibrary/MaterialFunctions/Basics/MF_UI_BooleanOperations}
  n12 = LinearInterpolate(A=@n11.IsGreater, B=OneMinus(None=@n11.IsGreater), Alpha=ScalarParameter() {parameterName=FlipColors})
  LinearInterpolate(A=LinearInterpolate(B=LinearInterpolate(A=@n9, B=@n10, Alpha=@n12), Alpha=@n8), Alpha=Min(A=@n5.Inner, B=@n7.Inner))
 ]
}
