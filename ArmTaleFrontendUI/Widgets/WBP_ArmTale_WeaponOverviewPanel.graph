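# CommonUI activatable panel for the ArmTale frontend showing a weapon's level, XP progress and a next mod reward preview.
# PreConstruct fills the reward's name, category and icon from PreviewNextModReward, tinting its dynamic material with rarity colors.
# HandleRarityChanged derives placeholder level and XP text from the rarity enum's integer plus the FAKE view model's selected-weapon progress.
# HandleLevelProgressChanged feeds the progress bar material's Progress parameter.
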WidgetBlueprint /ArmTaleFrontendUI/Widgets/WBP_ArmTale_WeaponOverviewPanel
graphs [
 EventGraph {nodes [
   Event PreConstruct()
   n1 = Get!() -> (
    PreviewNextModReward_ItemDefinition_34_#1:Fort Weapon Mod Item Definition Object Reference
    PreviewNextModReward_ModType_30_#2:E_ArmTale_ModTypes_FAKE Enum
    PreviewNextModReward_Stat0Buff_26_#3:double=~
    PreviewNextModReward_Stat1Buff_20_#4:double=~
    PreviewNextModReward_Stat2Buff_22_#5:double=~
    PreviewNextModReward_Stat3Buff_24_#6:double=~
   )
   > "SetText (Text)"(
    self=Get TextBlock_NextModName() -> (TextBlock_NextModName:Common Text Object Reference)
    InText=GetDisplayName!(self:Item Definition Base Object Reference=@n1) -> (ReturnValue:Text)
   )
   > "SetText (Text)"(
    self=Get TextBlock_NextModCategory() -> (TextBlock_NextModCategory:Common Text Object Reference)
    InText=To Text!(Value:E_ArmTale_ModTypes_FAKE Enum=@n1.PreviewNextModReward_ModType_30_#2) -> (ReturnValue:Text) {is=title K2Node_GenericToText}
   )
   > SetBrushFromLazyDisplayAsset!(
    execute
    self:Common Lazy Image Object Reference=Get Image_NextMod() -> (Image_NextMod:Fort Lazy Image Object Reference)
    LazyObject:"Soft Object Reference (by ref)"=GetLargePreviewImage!(ItemDefinition:Item Definition Base Object Reference=@n1) -> (ReturnValue:Texture 2D Soft Object Reference)
    bMatchTextureSize:bool=false
   ) -> (then)
   > n2 = GetRarityData!(
    execute
    Rarity:EFortRarity Enum=GetRarity!(ItemDefinition:Item Definition Base Object Reference=@n1) -> (ReturnValue:EFortRarity Enum=Common)
   ) -> (
    then
    ReturnValue_Name:Text
    ReturnValue_Color1:Linear Color Structure
    ReturnValue_Color2:Linear Color Structure
    ReturnValue_Color3:Linear Color Structure
    ReturnValue_Color4:Linear Color Structure
    ReturnValue_Color5:Linear Color Structure
    ReturnValue_Radius:float=~
    ReturnValue_Falloff:float=~
    ReturnValue_Brightness:float=~
    ReturnValue_Roughness:float=~
    ReturnValue_Glow:float=~
   )
   > n3 = GetDynamicMaterial(self=Get Image_NextMod())
   > SetVectorParameterValue(self=@n3, ParameterName=RarityColor_1, Value=@n2.ReturnValue_Color1)
   > SetVectorParameterValue(self=@n3, ParameterName=RarityColor_2, Value=@n2.ReturnValue_Color2)
   > SetVectorParameterValue(self=@n3, ParameterName=RarityColor_3, Value=@n2.ReturnValue_Color3)
  ]}
 HandleLevelProgressChanged {nodes [
   n1 = HandleLevelProgressChanged!() -> (then, LevelProgress:double=~) {is=title K2Node_FunctionEntry}
   > n2 = GetDynamicMaterial(self=Get Image_WeaponLevelProgressBar() -> (Image_WeaponLevelProgressBar:Image Object Reference))
   > SetScalarParameterValue(self=@n2, ParameterName=Progress, Value=@n1)
  ]}
 HandleRarityChanged {nodes [
   n1 = HandleRarityChanged!() -> (then, Rarity:EFortRarity Enum=Common) {is=title K2Node_FunctionEntry}
   n2 = "To Integer (Byte)"!(InByte:Byte=@n1) -> (ReturnValue:int=0)
   > n3 = Set WeaponLevel(WeaponLevel:int=int + int(A=@n2)) -> (Output_Get:int=0)
   > "SetText (Text)"(self=Get TextBlock_WeaponLevel() -> (TextBlock_WeaponLevel:Common Text Object Reference), InText="To Text (Integer)"(Value=@n3))
   n4 = Get VM_ArmTale_FAKE() -> (VM_ArmTale_FAKE:VM Arm Tale FAKE Object Reference)
   n5 = float + float(
    A:int=@n2
    B=Get!(self:VM Arm Tale Weapon Entry FAKE Object Reference=Get SelectedWeapon!(self:VM Arm Tale FAKE Object Reference=@n4) -> (SelectedWeapon:VM Arm Tale Weapon Entry FAKE Object Reference)) -> (
     WeaponData_ItemDefinition_64_#7:Fort Weapon Item Definition Object Reference
     WeaponData_WeaponLevelProgress_41_#2:double=~
     WeaponData_Stat0_43_#8:double=~
     WeaponData_Stat1_47_#9:double=~
     WeaponData_Stat2_48_#10:double=~
     WeaponData_Stat3_49_#11:double=~
     WeaponData_ModSlots_73_#12:Array of E_ArmTale_ModTypes_FAKE Enums
    ).WeaponData_WeaponLevelProgress_41_#2
   )
   > n6 = Set WeaponXPCurrent(WeaponXPCurrent:int=Truncate(A=float * float(A=@n5, B=1000) {is=title K2Node_PromotableOperator})) -> (Output_Get:int=0)
   > "SetText (Text)"(self=Get TextBlock_WeaponXPCurrent() -> (TextBlock_WeaponXPCurrent:Common Text Object Reference), InText="To Text (Integer)"(Value=@n6))
   > n7 = Set WeaponXPToNextLevel(WeaponXPToNextLevel:int=int * int(A=Get WeaponLevel() -> (WeaponLevel:int=0), B=1000)) -> (Output_Get:int=0)
   > "SetText (Text)"(
    self=Get TextBlock_WeaponXPToNextLevel() -> (TextBlock_WeaponXPToNextLevel:Common Text Object Reference)
    InText=Format Text!(Format="LOCTEXT(\"/{XPToNextLevel} EXP\")", XPToNextLevel:int=@n7) -> (Result)
   )
  ]}
]
variables [PreviewNextModReward {
  type=S_ArmTale_ModData_FAKE
  default {
   itemDefinition=/WeaponMods/Items/WeaponMods/Barrel/Suppressor/WMID_Barrel_Suppressor
   modType=Attachment
   stat0Buff=0
   stat1Buff=0
   stat2Buff=0
   stat3Buff=0
  }
 }]
widgets {
 info {inheritedWidgetCount=0, parentClass=&CommonActivatableWidget, rootWidgetClass=&Overlay, widgetCount=21}
 widgets [
  Overlay_0 {widgetClassPath=&Overlay, widget=Overlay_0}
  Image_PanelBackground {
   widgetClassPath=&Image
   parent=Overlay_0
   widget=Image_PanelBackground
   properties {brush {
     bIsDynamicallyLoaded=false
     drawAs=Image
     imageType=FullColor
     mirroring=NoMirror
     resourceName=None
     resourceObject=/ArmTaleFrontendUI/Assets/Materials/MI_UI_InfoPanelBG
     tiling=NoTile
     imageSize=<32 32>
     margin=<ltrb 0 0 0 0>
     outlineSettings {
      bUseBrushTransparency=false
      roundingType=HalfHeightRadius
      width=0
      color {colorUseRule=UseColor_Specified, specifiedColor=<rgba 0 0 0 0>}
      cornerRadii=<0 0 0 0>
     }
     tintColor {colorUseRule=UseColor_Specified, specifiedColor=<rgba 1 1 1 1>}
     uVRegion {bIsValid=false, max=<0 0>, min=<0 0>}
    }}
   slot {class=OverlaySlot, properties {horizontalAlignment=HAlign_Fill, verticalAlignment=VAlign_Fill}}
  }
  VerticalBox_106 {
   widgetClassPath=&VerticalBox
   parent=Overlay_0
   widget=VerticalBox_106
   properties {bOverride_Cursor=true}
   slot {class=OverlaySlot, properties {horizontalAlignment=HAlign_Fill, verticalAlignment=VAlign_Fill, padding=<ltrb 24 24 24 24>}}
  }
  HorizontalBox_47 {
   widgetClassPath=&HorizontalBox
   parent=VerticalBox_106
   widget=HorizontalBox_47
   slot {class=VerticalBoxSlot, properties {verticalAlignment=VAlign_Top}}
  }
  TextBlock_WeaponLevelLabel {
   widgetClassPath=&CommonTextBlock
   parent=HorizontalBox_47
   widget=TextBlock_WeaponLevelLabel
   properties {
    renderOpacity=0.4
    text=LVL
    textTransformPolicy=ToUpper
    font {
     bForceMonospaced=false
     bMaterialIsStencil=false
     fontMaterial=null
     fontObject=&Burbank
     letterSpacing=0
     monospacedWidth=1
     size=31.5
     skewAmount=0.2
     typefaceFontName=BurbankRegular-Black
     outlineSettings {
      bApplyOutlineToDropShadows=false
      bMiteredCorners=false
      bSeparateFillAlpha=false
      outlineMaterial=null
      outlineSize=0
      outlineColor=<rgba 0 0 0 1>
     }
    }
    shadowOffset=<0 0>
   }
   slot {class=HorizontalBoxSlot, properties {verticalAlignment=VAlign_Bottom, padding=<ltrb 0 0 8 0>}}
  }
  TextBlock_WeaponLevel {
   widgetClassPath=&CommonTextBlock
   bIsVariable=true
   parent=HorizontalBox_47
   widget=TextBlock_WeaponLevel
   properties {
    lineHeightPercentage=0.91
    text=2
    textTransformPolicy=ToUpper
    font {
     bForceMonospaced=false
     bMaterialIsStencil=false
     fontMaterial=null
     fontObject=&Burbank
     letterSpacing=0
     monospacedWidth=1
     size=51
     skewAmount=0.2
     typefaceFontName=BurbankRegular-Black
     outlineSettings {
      bApplyOutlineToDropShadows=false
      bMiteredCorners=false
      bSeparateFillAlpha=false
      outlineMaterial=null
      outlineSize=0
      outlineColor=<rgba 0 0 0 1>
     }
    }
    margin=<ltrb 0 -20 0 0>
    shadowOffset=<0 0>
   }
   slot {class=HorizontalBoxSlot, properties {verticalAlignment=VAlign_Bottom}}
  }
  Spacer_1 {
   widgetClassPath=&Spacer
   parent=HorizontalBox_47
   widget=Spacer_1
   slot {class=HorizontalBoxSlot, properties {size {sizeRule=Fill, value=1}}}
  }
  TextBlock_WeaponXPCurrent {
   widgetClassPath=&CommonTextBlock
   bIsVariable=true
   parent=HorizontalBox_47
   widget=TextBlock_WeaponXPCurrent
   properties {
    lineHeightPercentage=1.2
    text="4,200"
    textTransformPolicy=ToUpper
    font {
     bForceMonospaced=false
     bMaterialIsStencil=false
     fontMaterial=null
     fontObject=&BurbankSmall
     letterSpacing=0
     monospacedWidth=1
     size=18
     skewAmount=0
     typefaceFontName=Black
     outlineSettings {
      bApplyOutlineToDropShadows=false
      bMiteredCorners=false
      bSeparateFillAlpha=false
      outlineMaterial=null
      outlineSize=0
      outlineColor=<rgba 0 0 0 1>
     }
    }
    shadowOffset=<0 0>
   }
   slot {class=HorizontalBoxSlot, properties {horizontalAlignment=HAlign_Right, verticalAlignment=VAlign_Bottom}}
  }
  TextBlock_WeaponXPToNextLevel {
   widgetClassPath=&CommonTextBlock
   bIsVariable=true
   parent=HorizontalBox_47
   widget=TextBlock_WeaponXPToNextLevel
   properties {
    lineHeightPercentage=1.4
    renderOpacity=0.35
    text="/10,000 EXP"
    textTransformPolicy=ToUpper
    font {
     bForceMonospaced=false
     bMaterialIsStencil=false
     fontMaterial=null
     fontObject=&HeadingNow
     letterSpacing=0
     monospacedWidth=1
     size=13.5
     skewAmount=0
     typefaceFontName=75 Medium
     outlineSettings {
      bApplyOutlineToDropShadows=false
      bMiteredCorners=false
      bSeparateFillAlpha=false
      outlineMaterial=null
      outlineSize=0
      outlineColor=<rgba 0 0 0 1>
     }
    }
    shadowOffset=<0 0>
   }
   slot {class=HorizontalBoxSlot, properties {horizontalAlignment=HAlign_Right, verticalAlignment=VAlign_Bottom}}
  }
  Image_WeaponLevelProgressBar {
   widgetClassPath=&Image
   bIsVariable=true
   parent=VerticalBox_106
   widget=Image_WeaponLevelProgressBar
   properties {brush {
     bIsDynamicallyLoaded=false
     drawAs=Image
     imageType=NoImage
     mirroring=NoMirror
     resourceName=None
     resourceObject=/ArmTaleFrontendUI/Assets/Materials/M_UI_WeaponLevel_ProgressBar
     tiling=NoTile
     imageSize=<16 16>
     margin=<ltrb 0 0 0 0>
     outlineSettings {
      bUseBrushTransparency=false
      roundingType=HalfHeightRadius
      width=0
      color {colorUseRule=UseColor_Specified, specifiedColor=<rgba 0 0 0 0>}
      cornerRadii=<0 0 0 0>
     }
     tintColor {colorUseRule=UseColor_Specified, specifiedColor=<rgba 1 1 1 1>}
     uVRegion {bIsValid=false, max=<0 0>, min=<0 0>}
    }}
   slot {class=VerticalBoxSlot}
  }
  Spacer_2 {widgetClassPath=&Spacer, parent=VerticalBox_106, widget=Spacer_2, properties {size=<1 24>}, slot {class=VerticalBoxSlot}}
  HorizontalBox_0 {widgetClassPath=&HorizontalBox, parent=VerticalBox_106, widget=HorizontalBox_0, slot {class=VerticalBoxSlot}}
  Image_NextMod {
   widgetClassPath=&FortLazyImage
   bIsVariable=true
   parent=HorizontalBox_0
   widget=Image_NextMod
   properties {
    bOverride_Cursor=true
    materialTextureParamName=IconTexture
    visibility=Visible
    brush {
     bIsDynamicallyLoaded=false
     drawAs=Image
     imageType=FullColor
     mirroring=NoMirror
     resourceName=None
     resourceObject=/ArmTaleFrontendUI/Assets/Materials/MI_UI_ModTile_Preview
     tiling=NoTile
     imageSize=<100 100>
     margin=<ltrb 0 0 0 0>
     outlineSettings {
      bUseBrushTransparency=false
      roundingType=HalfHeightRadius
      width=0
      color {colorUseRule=UseColor_Specified, specifiedColor=<rgba 0 0 0 0>}
      cornerRadii=<0 0 0 0>
     }
     tintColor {colorUseRule=UseColor_Specified, specifiedColor=<rgba 1 1 1 1>}
     uVRegion {bIsValid=false, max=<0 0>, min=<0 0>}
    }
   }
   slot {class=HorizontalBoxSlot, properties {horizontalAlignment=HAlign_Center, verticalAlignment=VAlign_Center}}
  }
  Spacer_3 {widgetClassPath=&Spacer, parent=HorizontalBox_0, widget=Spacer_3, properties {size=<10 1>}, slot {class=HorizontalBoxSlot}}
  VerticalBox_0 {
   widgetClassPath=&VerticalBox
   parent=HorizontalBox_0
   widget=VerticalBox_0
   slot {class=HorizontalBoxSlot, properties {size {sizeRule=Fill, value=1}}}
  }
  TextBlock_NextModLabel {
   widgetClassPath=&CommonTextBlock
   parent=VerticalBox_0
   widget=TextBlock_NextModLabel
   properties {
    renderOpacity=0.35
    text=Next Reward
    font {
     bForceMonospaced=false
     bMaterialIsStencil=false
     fontMaterial=null
     fontObject=&HeadingNow
     letterSpacing=0
     monospacedWidth=1
     size=13.5
     skewAmount=0
     typefaceFontName=75 Medium
     outlineSettings {
      bApplyOutlineToDropShadows=false
      bMiteredCorners=false
      bSeparateFillAlpha=false
      outlineMaterial=null
      outlineSize=0
      outlineColor=<rgba 0 0 0 1>
     }
    }
    shadowOffset=<0 0>
   }
   slot {class=VerticalBoxSlot, properties {verticalAlignment=VAlign_Bottom, padding=<ltrb 0 0 0 8>}}
  }
  HorizontalBox_1 {
   widgetClassPath=&HorizontalBox
   parent=VerticalBox_0
   widget=HorizontalBox_1
   slot {class=VerticalBoxSlot, properties {verticalAlignment=VAlign_Bottom}}
  }
  Image_Bracket {
   widgetClassPath=&Image
   parent=HorizontalBox_1
   widget=Image_Bracket
   properties {
    renderOpacity=0.2
    brush {
     bIsDynamicallyLoaded=false
     drawAs=Image
     imageType=NoImage
     mirroring=NoMirror
     resourceName=None
     resourceObject=/ArmTaleFrontendUI/Assets/Materials/M_UI_SideBracket
     tiling=NoTile
     imageSize=<16 16>
     margin=<ltrb 0 0 0 0>
     outlineSettings {
      bUseBrushTransparency=false
      roundingType=HalfHeightRadius
      width=0
      color {colorUseRule=UseColor_Specified, specifiedColor=<rgba 0 0 0 0>}
      cornerRadii=<0 0 0 0>
     }
     tintColor {colorUseRule=UseColor_Specified, specifiedColor=<rgba 1 1 1 1>}
     uVRegion {bIsValid=false, max=<0 0>, min=<0 0>}
    }
   }
   slot {class=HorizontalBoxSlot}
  }
  VerticalBox_1 {
   widgetClassPath=&VerticalBox
   parent=HorizontalBox_1
   widget=VerticalBox_1
   slot {class=HorizontalBoxSlot, properties {padding=<ltrb 0 6 0 10>}}
  }
  TextBlock_NextModName {
   widgetClassPath=&CommonTextBlock
   bIsVariable=true
   parent=VerticalBox_1
   widget=TextBlock_NextModName
   properties {
    text="{Item Name}"
    textTransformPolicy=ToUpper
    font {
     bForceMonospaced=false
     bMaterialIsStencil=false
     fontMaterial=null
     fontObject=&HeadingNow
     letterSpacing=0
     monospacedWidth=1
     size=15
     skewAmount=0
     typefaceFontName=76 Bold
     outlineSettings {
      bApplyOutlineToDropShadows=false
      bMiteredCorners=false
      bSeparateFillAlpha=false
      outlineMaterial=null
      outlineSize=0
      outlineColor=<rgba 0 0 0 1>
     }
    }
    shadowOffset=<0 0>
   }
   slot {class=VerticalBoxSlot}
  }
  TextBlock_NextModCategory {
   widgetClassPath=&CommonTextBlock
   bIsVariable=true
   parent=VerticalBox_1
   widget=TextBlock_NextModCategory
   properties {
    text="{Item Category}"
    textTransformPolicy=ToUpper
    font {
     bForceMonospaced=false
     bMaterialIsStencil=false
     fontMaterial=null
     fontObject=&HeadingNow
     letterSpacing=0
     monospacedWidth=1
     size=12
     skewAmount=0
     typefaceFontName=74 Regular
     outlineSettings {
      bApplyOutlineToDropShadows=false
      bMiteredCorners=false
      bSeparateFillAlpha=false
      outlineMaterial=null
      outlineSize=0
      outlineColor=<rgba 0 0 0 1>
     }
    }
    shadowOffset=<0 0>
   }
   slot {class=VerticalBoxSlot}
  }
 ]
}
