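# Translucent UI material that procedurally draws an arrow, named for stat previews in the ArmTale frontend.
# Normalized U is mirrored around the center with Abs, recombined with V, then vertically skewed through MF_UI_Skew.
# The warped UVs feed MF_UI_RegularPolygon set to three sides, so the arrow shape comes from math with no texture.
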Material /ArmTaleFrontendUI/Assets/Materials/M_UI_PreviewStatArrow
graph {
 properties {bAutomaticallySetUsageInEditor=false, blendMode=BLEND_Translucent, materialDomain=MD_UI, usedShadingModels=Default Lit}
 nodes [
  Constant() {r=1}
  n1 = MaterialFunctionCall() {materialFunction=&GetUserInterfaceUV}
  n2 = ComponentMask(None=@n1.Normalized UV) {r=true}
  n3 = Add(A=Abs(None=Subtract(A=@n2) {constB=0.5})) {constB=0.5}
  n4 = ComponentMask(None=@n1.Normalized UV) {g=true}
  MaterialFunctionCall!(
   UVs=MaterialFunctionCall!(UV=AppendVector(A=@n3, B=@n4), Skew Angle=Constant2Vector() {g=-25}) {materialFunction=/FortUILibrary/MaterialFunctions/Basics/MF_UI_Skew}.Skewed UVs
   Size=Constant() {r=0.9}
   Sides=Constant() {r=3}
   InnerAngle
   OutlineThickness
   OutlineAlignment
   GlowMin
   GlowMax
  ) {materialFunction=/FortUILibrary/MaterialFunctions/Shapes/MF_UI_RegularPolygon}
 ]
}
