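# Weapon stat row widget for the ArmTale frontend, showing a label, progress bar and numeric value with a preview of a stat change.
# Setters cache base and preview values and drive the bar's dynamic material, then refresh text next tick via InvalidateText.
# A nonzero preview shows the zero-clamped sum with an up/down arrow, both tinted green or red.
# bFlipColors inverts that color meaning; PreConstruct applies preview variables and passes the flag to the material.
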WidgetBlueprint /ArmTaleFrontendUI/Widgets/WBP_ArmTale_WeaponStat
graphs [
 EventGraph {nodes [
   InvalidateText() {is=title K2Node_CustomEvent}
   > DelayUntilNextTick!(execute) -> (then)
   > n1 = Branch(Condition="Equal (Float)"(A=Get CachePreviewValue() -> (CachePreviewValue:double)))
   n2 = Get CacheBaseValue() -> (CacheBaseValue:double)
   > "SetText (Text)"(
    self=Get TextBlock_StatValue() -> (TextBlock_StatValue:Common Text Object Reference)
    InText="To Text (Float)"(Value="Max (Float)"(A=@n2, B=0), MinimumIntegralDigits=1, MaximumIntegralDigits=324, MinimumFractionalDigits=0, MaximumFractionalDigits=1)
   )
   > SetVisibility(self=Get Image_PreviewArrow() -> (Image_PreviewArrow:Image Object Reference), InVisibility=Hidden)
   > SetColorAndOpacity!(
    execute
    self=Get TextBlock_StatValue()
    InColorAndOpacity_SpecifiedColor:Linear Color Structure=(R=1,G=1,B=1,A=1)
    InColorAndOpacity_ColorUseRule:ESlateColorStylingMode Enum
   ) -> (then)
   n3 = float + float(A=Get CacheBaseValue(), B=Get CachePreviewValue())
   "SetText (Text)"(
    execute=@n1.else
    self=Get TextBlock_StatValue()
    InText="To Text (Float)"(Value="Max (Float)"(A=@n3, B=0), MinimumIntegralDigits=1, MaximumIntegralDigits=324, MinimumFractionalDigits=0, MaximumFractionalDigits=1)
   )
   > SetVisibility(self=Get Image_PreviewArrow(), InVisibility=SelfHitTestInvisible)
   n4 = "float >= float"(A=Get CachePreviewValue())
   > SetRenderScale!(execute, self:Widget Object Reference=Get Image_PreviewArrow(), Scale_X:double=1, Scale_Y:double=SelectFloat(A=1, B=-1, bPickA=@n4)) -> (then)
   n5 = Get bFlipColors() -> (bFlipColors:bool)
   n6 = SelectColor(
    A=(R=0.0331,G=1,B=0.2159,A=1)
    B=(R=1,G=0.013,B=0.0931,A=1)
    bPickA=Select(
     Option 0:bool=@n4
     Option 1:bool=NOT Boolean(A=@n4)
     Index="Equal (Boolean)"(A:bool=@n5, B:bool=true) -> (ReturnValue:bool=~) {is=title K2Node_PromotableOperator}
    ) -> (ReturnValue:bool)
   )
   > SetColorAndOpacity!(execute, self:Image Object Reference=Get Image_PreviewArrow(), InColorAndOpacity:Linear Color Structure=@n6) -> (then)
   > SetColorAndOpacity(self:Text Object Reference=Get TextBlock_StatValue(), InColorAndOpacity:Slate Color Structure=Make Slate Color(SpecifiedColor=@n6))
   Event PreConstruct()
   > SetStatName!(execute, self, Name:Text=Get PreviewStatName() -> (PreviewStatName:Text)) -> (then)
   > SetMaxValue!(execute, self, MaxValue:double=Get PreviewMaxValue() -> (PreviewMaxValue:double)) -> (then)
   > SetBaseValue!(execute, self, Value:double=Get PreviewBaseValue() -> (PreviewBaseValue:double)) -> (then)
   > SetPreviewValue!(execute, self, Value:double=Get PreviewStatValue() -> (PreviewStatValue:double)) -> (then)
   > n7 = GetDynamicMaterial(self=Get Image_ProgressBar() -> (Image_ProgressBar:Image Object Reference))
   > SetScalarParameterValue(self=@n7, ParameterName=FlipColors, Value="To Float (Boolean)"(InBool=Get bFlipColors()))
  ]}
 SetBaseValue {nodes [
   n1 = SetBaseValue!() -> (then, Value:double=~) {is=title K2Node_FunctionEntry}
   > n2 = Set CacheBaseValue(CacheBaseValue:double=@n1) -> (Output_Get:double)
   > n3 = GetDynamicMaterial(self=Get Image_ProgressBar())
   > SetScalarParameterValue(self=@n3, ParameterName=Progress, Value=@n2)
   > InvalidateText() {is=title K2Node_CallFunction}
  ]}
 SetMaxValue {nodes [
   n1 = SetMaxValue!() -> (then, MaxValue:double=~) {is=title K2Node_FunctionEntry}
   > n2 = GetDynamicMaterial(self=Get Image_ProgressBar())
   > SetScalarParameterValue(self=@n2, ParameterName=MaxValue, Value=@n1)
  ]}
 SetPreviewValue {nodes [
   n1 = SetPreviewValue!() -> (then, Value:double=~) {is=title K2Node_FunctionEntry}
   > n2 = Set CachePreviewValue(CachePreviewValue:double=@n1) -> (Output_Get:double)
   > n3 = GetDynamicMaterial(self=Get Image_ProgressBar())
   > SetScalarParameterValue(self=@n3, ParameterName=PreviewValue, Value=@n2)
   > InvalidateText()
  ]}
 SetStatName {nodes [
   n1 = SetStatName!() -> (then, Name:Text) {is=title K2Node_FunctionEntry}
   > "SetText (Text)"(self=Get TextBlock_StatLabel() -> (TextBlock_StatLabel:Common Text Object Reference), InText=@n1)
  ]}
]
variables [
 bFlipColors {type=boolean}
 CacheBaseValue {type=number}
 CachePreviewValue {type=number}
 PreviewBaseValue {type=number, default=40}
 PreviewMaxValue {type=number, default=100}
 PreviewStatName {type=string, default="{Stat Name}"}
 PreviewStatValue {type=number}
]
widgets {
 info {inheritedWidgetCount=0, parentClass=&UserWidget, rootWidgetClass=&HorizontalBox, widgetCount=6}
 widgets [
  HorizontalBox_0 {widgetClassPath=&HorizontalBox, widget=HorizontalBox_0, properties {bOverride_Cursor=true}}
  TextBlock_StatLabel {
   widgetClassPath=&CommonTextBlock
   bIsVariable=true
   parent=HorizontalBox_0
   widget=TextBlock_StatLabel
   properties {
    minDesiredWidth=180
    text="{Stat Label}"
    textTransformPolicy=ToUpper
    font {
     bForceMonospaced=false
     bMaterialIsStencil=false
     fontMaterial=null
     fontObject=&HeadingNow
     letterSpacing=0
     monospacedWidth=1
     size=12
     skewAmount=0
     typefaceFontName=74 Regular
     outlineSettings {
      bApplyOutlineToDropShadows=false
      bMiteredCorners=false
      bSeparateFillAlpha=false
      outlineMaterial=null
      outlineSize=0
      outlineColor=<rgba 0 0 0 1>
     }
    }
    shadowOffset=<0 0>
   }
   slot {class=HorizontalBoxSlot, properties {verticalAlignment=VAlign_Center}}
  }
  Image_ProgressBar {
   widgetClassPath=&Image
   bIsVariable=true
   parent=HorizontalBox_0
   widget=Image_ProgressBar
   properties {brush {
     bIsDynamicallyLoaded=false
     drawAs=Image
     imageType=FullColor
     mirroring=NoMirror
     resourceName=None
     resourceObject=/ArmTaleFrontendUI/Assets/Materials/M_UI_WeaponLevel_ProgressBar
     tiling=NoTile
     imageSize=<6 6>
     margin=<ltrb 0 0 0 0>
     outlineSettings {
      bUseBrushTransparency=false
      roundingType=HalfHeightRadius
      width=0
      color {colorUseRule=UseColor_Specified, specifiedColor=<rgba 0 0 0 0>}
      cornerRadii=<0 0 0 0>
     }
     tintColor {colorUseRule=UseColor_Specified, specifiedColor=<rgba 1 1 1 1>}
     uVRegion {bIsValid=false, max=<0 0>, min=<0 0>}
    }}
   slot {class=HorizontalBoxSlot, properties {verticalAlignment=VAlign_Center, padding=<ltrb 8 0 8 0>, size {sizeRule=Fill, value=1}}}
  }
  Overlay_0 {
   widgetClassPath=&Overlay
   parent=HorizontalBox_0
   widget=Overlay_0
   slot {class=HorizontalBoxSlot, properties {verticalAlignment=VAlign_Center}}
  }
  TextBlock_StatValue {
   widgetClassPath=&CommonTextBlock
   bIsVariable=true
   parent=Overlay_0
   widget=TextBlock_StatValue
   properties {
    justification=Right
    minDesiredWidth=40
    text=40
    textTransformPolicy=ToUpper
    font {
     bForceMonospaced=false
     bMaterialIsStencil=false
     fontMaterial=null
     fontObject=&HeadingNow
     letterSpacing=0
     monospacedWidth=1
     size=13.5
     skewAmount=0
     typefaceFontName=75 Medium
     outlineSettings {
      bApplyOutlineToDropShadows=false
      bMiteredCorners=false
      bSeparateFillAlpha=false
      outlineMaterial=null
      outlineSize=0
      outlineColor=<rgba 0 0 0 1>
     }
    }
    shadowOffset=<0 0>
   }
   slot {class=OverlaySlot}
  }
  Image_PreviewArrow {
   widgetClassPath=&Image
   bIsVariable=true
   parent=Overlay_0
   widget=Image_PreviewArrow
   properties {
    brush {
     bIsDynamicallyLoaded=false
     drawAs=Image
     imageType=NoImage
     mirroring=NoMirror
     resourceName=None
     resourceObject=/ArmTaleFrontendUI/Assets/Materials/M_UI_PreviewStatArrow
     tiling=NoTile
     imageSize=<12 12>
     margin=<ltrb 0 0 0 0>
     outlineSettings {
      bUseBrushTransparency=false
      roundingType=HalfHeightRadius
      width=0
      color {colorUseRule=UseColor_Specified, specifiedColor=<rgba 0 0 0 0>}
      cornerRadii=<0 0 0 0>
     }
     tintColor {colorUseRule=UseColor_Specified, specifiedColor=<rgba 1 1 1 1>}
     uVRegion {bIsValid=false, max=<0 0>, min=<0 0>}
    }
    renderTransform {angle=0, scale=<1 1>, shear=<0 0>, translation=<15 0>}
   }
   slot {class=OverlaySlot, properties {horizontalAlignment=HAlign_Right, verticalAlignment=VAlign_Center}}
  }
 ]
}
