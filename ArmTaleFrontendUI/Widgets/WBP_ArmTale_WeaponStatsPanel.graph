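# Weapon stats panel for the ArmTale frontend, stacking four stat widgets previewed as Damage, Fire Rate, Clip Size and Reload Time (colors flipped).
# HandleEquippedModsChanged sums each equipped mod's buffs and adds them to the viewmodel's selected weapon stats as base values.
# HandlePreviewModChanged shows per-stat differences between the preview mod and the selected slot's equipped mod.
# Previews are zeroed when that mod is invalid or already equipped.
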WidgetBlueprint /ArmTaleFrontendUI/Widgets/WBP_ArmTale_WeaponStatsPanel
graphs [
 EventGraph {nodes [
   Event PreConstruct()
   Event Construct()
   Event Tick()
  ]}
 HandleEquippedModsChanged {nodes [
   n1 = HandleEquippedModsChanged!() -> (then, EquippedMods:Array of VM Arm Tale Mod Entry FAKE Object References) {is=title K2Node_FunctionEntry}
   > n2 = For Each Loop(Array:Array of VM Arm Tale Mod Entry FAKE Object References=@n1) -> (Array Element:VM Arm Tale Mod Entry FAKE Object Reference)
   n3 = Sequence!(execute=@n2.Completed) -> (then_0, then_1, then_2:Exec, then_3:Exec)
   n4 = Get!(self:VM Arm Tale Weapon Entry FAKE Object Reference=Get SelectedWeapon!(self:VM Arm Tale FAKE Object Reference=Get VM_ArmTale_FAKE() -> (VM_ArmTale_FAKE:VM Arm Tale FAKE Object Reference)) -> (SelectedWeapon:VM Arm Tale Weapon Entry FAKE Object Reference)) -> (
    WeaponData_ItemDefinition_64_#1:Fort Weapon Item Definition Object Reference
    WeaponData_WeaponLevelProgress_41_#2:double=~
    WeaponData_Stat0_43_#3:double=~
    WeaponData_Stat1_47_#4:double=~
    WeaponData_Stat2_48_#5:double=~
    WeaponData_Stat3_49_#6:double=~
    WeaponData_ModSlots_73_#7:Array of E_ArmTale_ModTypes_FAKE Enums
   )
   SetBaseValue!(
    execute=@n3.then_2
    self:WBP Arm Tale Weapon Stat Object Reference=Get WeaponStat3() -> (WeaponStat3:WBP Arm Tale Weapon Stat Object Reference)
    Value:double=float + float(A=@n4.WeaponData_Stat2_48_#5, B=Get CollatedStat2Buff() -> (CollatedStat2Buff:double))
   ) -> (then)
   SetBaseValue(
    execute=@n3.then_3
    self=Get WeaponStat4() -> (WeaponStat4:WBP Arm Tale Weapon Stat Object Reference)
    Value=float + float(A=@n4.WeaponData_Stat3_49_#6, B=Get CollatedStat3Buff() -> (CollatedStat3Buff:double))
   )
   SetBaseValue(
    execute=@n3.then_0
    self=Get WeaponStat1() -> (WeaponStat1:WBP Arm Tale Weapon Stat Object Reference)
    Value=float + float(A=@n4.WeaponData_Stat0_43_#3, B=Get CollatedStat0Buff() -> (CollatedStat0Buff:double))
   )
   SetBaseValue(
    execute=@n3.then_1
    self=Get WeaponStat2() -> (WeaponStat2:WBP Arm Tale Weapon Stat Object Reference)
    Value=float + float(A=@n4.WeaponData_Stat1_47_#4, B=Get CollatedStat1Buff() -> (CollatedStat1Buff:double))
   )
   n5 = Sequence(execute=@n2.LoopBody)
   n6 = Get!(self:VM Arm Tale Mod Entry FAKE Object Reference=@n2) -> (
    ModData_ItemDefinition_34_#8:Fort Weapon Mod Item Definition Object Reference
    ModData_ModType_30_#2:E_ArmTale_ModTypes_FAKE Enum
    ModData_Stat0Buff_26_#9:double=~
    ModData_Stat1Buff_20_#10:double=~
    ModData_Stat2Buff_22_#11:double=~
    ModData_Stat3Buff_24_#12:double=~
   )
   > Set CollatedStat0Buff(CollatedStat0Buff:double=float + float(A=Get CollatedStat0Buff(), B=@n6.ModData_Stat0Buff_26_#9)) -> (Output_Get:double)
   Set CollatedStat1Buff(
    execute=@n5.then_1
    CollatedStat1Buff:double=float + float(A=Get CollatedStat1Buff(), B=@n6.ModData_Stat1Buff_20_#10)
   ) -> (Output_Get:double)
   Set CollatedStat2Buff(
    execute=@n5.then_2
    CollatedStat2Buff:double=float + float(A=Get CollatedStat2Buff(), B=@n6.ModData_Stat2Buff_22_#11)
   ) -> (Output_Get:double)
   Set CollatedStat3Buff(
    execute=@n5.then_3
    CollatedStat3Buff:double=float + float(A=Get CollatedStat3Buff(), B=@n6.ModData_Stat3Buff_24_#12)
   ) -> (Output_Get:double)
  ]}
 HandlePreviewModChanged {nodes [
   n1 = HandlePreviewModChanged!() -> (then, PreviewMod:VM Arm Tale Mod Entry FAKE Object Reference) {is=title K2Node_FunctionEntry}
   > n2 = Is Valid(InputObject=@n1)
   > n3 = Branch(Condition=Get IsEquipped!(self:VM Arm Tale Mod Entry FAKE Object Reference=@n1) -> (IsEquipped:bool))
   n4 = Sequence(execute=[@n3.then, @n2.Is Not Valid])
   SetPreviewValue!(execute=@n4.then_1, self:WBP Arm Tale Weapon Stat Object Reference=Get WeaponStat2(), Value:double=0) -> (then)
   SetPreviewValue(execute=@n4.then_0, self=Get WeaponStat1(), Value=0)
   SetPreviewValue(execute=@n4.then_2, self=Get WeaponStat3(), Value=0)
   SetPreviewValue(execute=@n4.then_3, self=Get WeaponStat4(), Value=0)
   n5 = GetEquippedModForSelectedSlot!(execute=@n3.else, self:VM Arm Tale FAKE Object Reference=Get VM_ArmTale_FAKE()) -> (then, EquippedMod:VM Arm Tale Mod Entry FAKE Object Reference)
   > Set EquippedModData(EquippedModData:S Arm Tale Mod Data FAKE Structure=Get ModData!(self:VM Arm Tale Mod Entry FAKE Object Reference=@n5) -> (ModData:S Arm Tale Mod Data FAKE Structure)) -> (Output_Get:S Arm Tale Mod Data FAKE Structure)
   > n6 = Sequence()
   n7 = Get(self=Get PreviewMod() -> (PreviewMod:VM Arm Tale Mod Entry FAKE Object Reference)) -> (
    ModData_Stat0Buff_26_#9=~
    ModData_Stat1Buff_20_#10=~
    ModData_Stat2Buff_22_#11=~
    ModData_Stat3Buff_24_#12=~
   )
   n8 = Get!() -> (
    EquippedModData_ItemDefinition_34_#8:Fort Weapon Mod Item Definition Object Reference
    EquippedModData_ModType_30_#2:E_ArmTale_ModTypes_FAKE Enum
    EquippedModData_Stat0Buff_26_#9:double=~
    EquippedModData_Stat1Buff_20_#10:double=~
    EquippedModData_Stat2Buff_22_#11:double=~
    EquippedModData_Stat3Buff_24_#12:double=~
   )
   SetPreviewValue(
    execute=@n6.then_1
    self=Get WeaponStat2()
    Value=float - float(A=@n7.ModData_Stat1Buff_20_#10, B=@n8.EquippedModData_Stat1Buff_20_#10)
   )
   SetPreviewValue(
    execute=@n6.then_0
    self=Get WeaponStat1()
    Value=float - float(A=@n7.ModData_Stat0Buff_26_#9, B=@n8.EquippedModData_Stat0Buff_26_#9)
   )
   SetPreviewValue(
    execute=@n6.then_2
    self=Get WeaponStat3()
    Value=float - float(A=@n7.ModData_Stat2Buff_22_#11, B=@n8.EquippedModData_Stat2Buff_22_#11)
   )
   SetPreviewValue(
    execute=@n6.then_3
    self=Get WeaponStat4()
    Value=float - float(A=@n7.ModData_Stat3Buff_24_#12, B=@n8.EquippedModData_Stat3Buff_24_#12)
   )
  ]}
]
variables []
widgets {
 info {inheritedWidgetCount=0, parentClass=&UserWidget, rootWidgetClass=&VerticalBox, widgetCount=5}
 widgets [
  VerticalBox_12 {widgetClassPath=&VerticalBox, widget=VerticalBox_12}
  WeaponStat1 {
   widgetClassPath=&/WBP_ArmTale_WeaponStat
   bIsVariable=true
   parent=VerticalBox_12
   widget=WeaponStat1
   properties {previewBaseValue=10, previewStatName=Damage}
   slot {class=VerticalBoxSlot, properties {verticalAlignment=VAlign_Center, padding=<ltrb 0 4 0 4>}}
  }
  WeaponStat2 {
   widgetClassPath=&WBP_ArmTale_WeaponStat
   bIsVariable=true
   parent=VerticalBox_12
   widget=WeaponStat2
   properties {previewBaseValue=110, previewMaxValue=150, previewStatName=Fire Rate}
   slot {class=VerticalBoxSlot, properties {verticalAlignment=VAlign_Center, padding=<ltrb 0 4 0 4>}}
  }
  WeaponStat3 {
   widgetClassPath=&WBP_ArmTale_WeaponStat
   bIsVariable=true
   parent=VerticalBox_12
   widget=WeaponStat3
   properties {previewBaseValue=60, previewStatName=Clip Size}
   slot {class=VerticalBoxSlot, properties {verticalAlignment=VAlign_Center, padding=<ltrb 0 4 0 4>}}
  }
  WeaponStat4 {
   widgetClassPath=&WBP_ArmTale_WeaponStat
   bIsVariable=true
   parent=VerticalBox_12
   widget=WeaponStat4
   properties {bFlipColors=true, previewBaseValue=1.8, previewMaxValue=10, previewStatName=Reload Time}
   slot {class=VerticalBoxSlot, properties {verticalAlignment=VAlign_Center, padding=<ltrb 0 4 0 4>}}
  }
 ]
}
