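# ArmTale frontend weapon details screen: an activatable widget showing weapon name, rarity, stats and preview beneath a top tab bar.
# On activation it sets the CosmeticDisplay camera, activates the item preview, binds tab changes and selects Showcase.
# Mod-slot tabs show mod selection and set the slot index on a FAKE viewmodel; other tabs show the overview and reset it.
# Rarity changes tint the background glow; Back clears SelectedWeapon, returning false.
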WidgetBlueprint /ArmTaleFrontendUI/Widgets/WBP_ArmTale_WeaponDetails_Screen
graphs [
 BP_GetDesiredFocusTarget {nodes [
   Get Desired Focus Target()
   > Return Node!(
    execute
    ReturnValue:Widget Object Reference=GetActiveWidget!(self:Widget Switcher Object Reference=Get Switcher_Content() -> (Switcher_Content:Common Activatable Widget Switcher Object Reference)) -> (ReturnValue:Widget Object Reference)
   )
  ]}
 BP_OnHandleBackAction {nodes [
   On Handle Back Action() {is=title K2Node_FunctionEntry}
   > Set with Broadcast SelectedWeapon!(
    execute
    SelectedWeapon:VM Arm Tale Weapon Entry FAKE Object Reference
    self:VM Arm Tale FAKE Object Reference=Get VM_ArmTale_FAKE() -> (VM_ArmTale_FAKE:VM Arm Tale FAKE Object Reference)
   ) -> (then, Output_Get:VM Arm Tale Weapon Entry FAKE Object Reference)
   > Return Node(ReturnValue:bool=false)
  ]}
 EventGraph {nodes [
   n1 = HandleOnSelectedTabChanged!() -> (OutputDelegate, then, AssociatedButton:Common Button Base Object Reference, ButtonIndex:int) {is=title K2Node_CustomEvent}
   > n2 = Cast To WBP_ArmTale_ModSlot!(execute, Object=@n1.AssociatedButton) -> (then, CastFailed, AsWBP Arm Tale Mod Slot:WBP Arm Tale Mod Slot Object Reference) {is=title K2Node_DynamicCast}
   n3 = Get Switcher_Content()
   > SetActiveWidgetIndex(self=@n3, Index=1)
   n4 = Get VM_ArmTale_FAKE()
   > SetSelectedModSlotIndex!(
    execute
    self:VM Arm Tale FAKE Object Reference=@n4
    ModSlotIndex:int=Get ModSlotIndex!(self:WBP Arm Tale Mod Slot Object Reference=@n2) -> (ModSlotIndex:int=0)
   ) -> (then)
   SetActiveWidgetIndex(execute=@n2.CastFailed, self=@n3, Index=0)
   > SetSelectedModSlotIndex(self=@n4, ModSlotIndex=-1)
   Event On Activated()
   > SetFrontEndCamera!(
    execute
    self:Fort Front End Context Object Reference=Get FortFrontEndContext() -> (ReturnValue:Fort Front End Context Object Reference) {is=title K2Node_GetSubsystem}
    NewState:EFrontEndCamera Enum=CosmeticDisplay
    TransitionParams:View Target Transition Params Structure=(BlendTime=0,BlendFunction=VTBlend_Cubic,BlendExp=2,bLockOutgoing=False)
   ) -> (then)
   > ActivateWidget(self=Get FortItemPreviewWidget() -> (FortItemPreviewWidget:WBP Item Preview Object Reference))
   > Bind Event to On Selected Button Base Changed(
    self:Common Button Group Base Object Reference=Get ButtonGroup!(self:WBP Arm Tale Weapon Details Top Nav Object Reference=Get TabList_TopNav() -> (TabList_TopNav:WBP Arm Tale Weapon Details Top Nav Object Reference)) -> (ButtonGroup:Common Button Group Base Object Reference)
    Delegate=Create Event()
   ) {is=title K2Node_AddDelegate}
   > HandleOnSelectedTabChanged!(
    execute
    self
    AssociatedButton:Common Button Base Object Reference=Get Button_Showcase!(self:WBP Arm Tale Weapon Details Top Nav Object Reference=Get TabList_TopNav()) -> (Button_Showcase:WBP Arm Tale Tab Button Object Reference)
    ButtonIndex:int=0
   ) -> (then) {is=title K2Node_CallFunction}
   Event On Deactivated()
   > DeactivateWidget(self=Get FortItemPreviewWidget())
   > Unbind Event from On Selected Button Base Changed(self:Common Button Group Base Object Reference=Get ButtonGroup(self=Get TabList_TopNav()), Delegate=Create Event()) {is=title K2Node_RemoveDelegate}
  ]}
 HandleRarityChanged {nodes [
   n1 = HandleRarityChanged!() -> (then, Rarity:EFortRarity Enum=Common) {is=title K2Node_FunctionEntry}
   > n2 = SetRarity!(
    execute
    self:WBP Arm Tale Rarity Label Object Reference=Get RarityLabel_Weapon() -> (RarityLabel_Weapon:WBP Arm Tale Rarity Label Object Reference)
    Rarity:EFortRarity Enum=@n1
   ) -> (
    then
    RarityData_Name:Text
    RarityData_Color1:Linear Color Structure
    RarityData_Color2:Linear Color Structure
    RarityData_Color3:Linear Color Structure
    RarityData_Color4:Linear Color Structure
    RarityData_Color5:Linear Color Structure
    RarityData_Radius:float=~
    RarityData_Falloff:float=~
    RarityData_Brightness:float=~
    RarityData_Roughness:float=~
    RarityData_Glow:float=~
   )
   > n3 = GetDynamicMaterial(self=Get Image_WeaponBGGlow() -> (Image_WeaponBGGlow:Image Object Reference))
   > SetVectorParameterValue(self=@n3, ParameterName=Rarity_1, Value=@n2.RarityData_Color2)
   > SetVectorParameterValue(self=@n3, ParameterName=Rarity_2, Value=@n2.RarityData_Color4)
  ]}
]
variables []
widgets {
 info {inheritedWidgetCount=0, parentClass=&CommonActivatableWidget, rootWidgetClass=&Overlay, widgetCount=19}
 widgets [
  Overlay_13 {widgetClassPath=&Overlay, widget=Overlay_13}
  Image_Background {
   widgetClassPath=&Image
   parent=Overlay_13
   widget=Image_Background
   properties {brush {
     bIsDynamicallyLoaded=false
     drawAs=Image
     imageType=NoImage
     mirroring=NoMirror
     resourceName=None
     resourceObject=/ArmTaleFrontendUI/Assets/Materials/M_UI_ScreenBG
     tiling=NoTile
     imageSize=<32 32>
     margin=<ltrb 0 0 0 0>
     outlineSettings {
      bUseBrushTransparency=false
      roundingType=HalfHeightRadius
      width=0
      color {colorUseRule=UseColor_Specified, specifiedColor=<rgba 0 0 0 0>}
      cornerRadii=<0 0 0 0>
     }
     tintColor {colorUseRule=UseColor_Specified, specifiedColor=<rgba 1 1 1 1>}
     uVRegion {bIsValid=false, max=<0 0>, min=<0 0>}
    }}
   slot {class=OverlaySlot, properties {horizontalAlignment=HAlign_Fill, verticalAlignment=VAlign_Fill}}
  }
  FortItemPreviewWidget {
   widgetClassPath=/FortniteCoreUI/Frontend/ItemPreview/WBP_ItemPreview
   bIsVariable=true
   parent=Overlay_13
   widget=FortItemPreviewWidget
   properties {
    bRestrictZoomToWidget=true
    bSpecialPreviewActionsEnabled=false
    bSupportsActivationFocus=false
    bTogglePreviewActionsEnabled=false
    infoAlignment_V=VAlign_Bottom
    infoWidget=&self:WidgetTree.Widget_PreviewInfo_External
    itemViewObject=&self:WidgetTree.FortItemPreviewWidget.ItemViewObject
    visibility=Visible
    defaultPreviewContextTags {gameplayTags [{tagName=UI.ItemPreview.PreviewType.OnPawn}]}
    screenTag {tagName=UI.ItemPreview.Screen.Locker}
   }
   slot {class=OverlaySlot, properties {horizontalAlignment=HAlign_Fill, verticalAlignment=VAlign_Fill}}
  }
  HorizontalBox_46 {
   widgetClassPath=&HorizontalBox
   parent=Overlay_13
   widget=HorizontalBox_46
   slot {class=OverlaySlot, properties {horizontalAlignment=HAlign_Fill, verticalAlignment=VAlign_Fill, padding=<ltrb 80 0 80 0>}}
  }
  SizeBox_0 {
   widgetClassPath=&SizeBox
   parent=HorizontalBox_46
   widget=SizeBox_0
   properties {bOverride_WidthOverride=true, widthOverride=500}
   slot {class=HorizontalBoxSlot, properties {verticalAlignment=VAlign_Top, padding=<ltrb 0 180 0 80>}}
  }
  VerticalBox_104 {
   widgetClassPath=&VerticalBox
   parent=SizeBox_0
   widget=VerticalBox_104
   slot {class=SizeBoxSlot, properties {verticalAlignment=VAlign_Top}}
  }
  TextBlock_WeaponName {
   widgetClassPath=&CommonTextBlock
   bIsVariable=true
   parent=VerticalBox_104
   widget=TextBlock_WeaponName
   properties {
    text="{Item Name}"
    textTransformPolicy=ToUpper
    font {
     bForceMonospaced=false
     bMaterialIsStencil=false
     fontMaterial=null
     fontObject=&Burbank
     letterSpacing=0
     monospacedWidth=1
     size=34.5
     skewAmount=0
     typefaceFontName=BurbankRegular-Black
     outlineSettings {
      bApplyOutlineToDropShadows=false
      bMiteredCorners=false
      bSeparateFillAlpha=false
      outlineMaterial=null
      outlineSize=0
      outlineColor=<rgba 0 0 0 1>
     }
    }
    shadowOffset=<0 0>
   }
   slot {class=VerticalBoxSlot}
  }
  RarityLabel_Weapon {
   widgetClassPath=&/WBP_ArmTale_RarityLabel
   bIsVariable=true
   parent=VerticalBox_104
   widget=RarityLabel_Weapon
   slot {class=VerticalBoxSlot, properties {horizontalAlignment=HAlign_Left, verticalAlignment=VAlign_Top}}
  }
  Spacer_78 {widgetClassPath=&Spacer, parent=VerticalBox_104, widget=Spacer_78, properties {size=<1 16>}, slot {class=VerticalBoxSlot}}
  WBP_ArmTale_WeaponStatsPanel {
   widgetClassPath=&/WBP_ArmTale_WeaponStatsPanel
   bIsVariable=true
   parent=VerticalBox_104
   widget=WBP_ArmTale_WeaponStatsPanel
   slot {class=VerticalBoxSlot}
  }
  Spacer {widgetClassPath=&Spacer, parent=VerticalBox_104, widget=Spacer, properties {size=<1 48>}, slot {class=VerticalBoxSlot}}
  Switcher_Content {
   widgetClassPath=&CommonActivatableWidgetSwitcher
   bIsVariable=true
   parent=VerticalBox_104
   widget=Switcher_Content
   properties {transitionDuration=0.3}
   slot {class=VerticalBoxSlot}
  }
  WBP_ArmTale_WeaponOverviewPanel {
   widgetClassPath=&/WBP_ArmTale_WeaponOverviewPanel
   bIsVariable=true
   parent=Switcher_Content
   widget=WBP_ArmTale_WeaponOverviewPanel
   slot {class=WidgetSwitcherSlot}
  }
  WBP_ArmTale_ModSelectionPanel {
   widgetClassPath=&/WBP_ArmTale_ModSelectionPanel
   bIsVariable=true
   parent=Switcher_Content
   widget=WBP_ArmTale_ModSelectionPanel
   slot {class=WidgetSwitcherSlot}
  }
  Overlay_0 {
   widgetClassPath=&Overlay
   parent=HorizontalBox_46
   widget=Overlay_0
   slot {class=HorizontalBoxSlot, properties {horizontalAlignment=HAlign_Center, verticalAlignment=VAlign_Center, size {sizeRule=Fill, value=1}}}
  }
  Image_WeaponBGGlow {
   widgetClassPath=&Image
   bIsVariable=true
   parent=Overlay_0
   widget=Image_WeaponBGGlow
   properties {
    renderOpacity=0.8
    visibility=SelfHitTestInvisible
    brush {
     bIsDynamicallyLoaded=false
     drawAs=Image
     imageType=NoImage
     mirroring=NoMirror
     resourceName=None
     resourceObject=/ArmTaleFrontendUI/Assets/Materials/M_UI_WeaponBGGlow
     tiling=NoTile
     imageSize=<32 32>
     margin=<ltrb 0 0 0 0>
     outlineSettings {
      bUseBrushTransparency=false
      roundingType=HalfHeightRadius
      width=0
      color {colorUseRule=UseColor_Specified, specifiedColor=<rgba 0 0 0 0>}
      cornerRadii=<0 0 0 0>
     }
     tintColor {colorUseRule=UseColor_Specified, specifiedColor=<rgba 1 1 1 1>}
     uVRegion {bIsValid=false, max=<0 0>, min=<0 0>}
    }
    renderTransform {angle=0, scale=<50 32>, shear=<0 0>, translation=<0 0>}
   }
   slot {class=OverlaySlot, properties {horizontalAlignment=HAlign_Center, verticalAlignment=VAlign_Center}}
  }
  Image_ItemPreview {
   widgetClassPath=&FortLazyImage
   bIsVariable=true
   parent=Overlay_0
   widget=Image_ItemPreview
   properties {
    brush {
     bIsDynamicallyLoaded=false
     drawAs=Image
     imageType=FullColor
     mirroring=NoMirror
     resourceName=None
     resourceObject=/Game/UI/Foundation/Textures/Icons/Weapons/Items/T-Icon-Weapons-PistolMark2TT-L
     tiling=NoTile
     imageSize=<640 640>
     margin=<ltrb 0 0 0 0>
     outlineSettings {
      bUseBrushTransparency=false
      roundingType=HalfHeightRadius
      width=0
      color {colorUseRule=UseColor_Specified, specifiedColor=<rgba 0 0 0 0>}
      cornerRadii=<0 0 0 0>
     }
     tintColor {colorUseRule=UseColor_Specified, specifiedColor=<rgba 1 1 1 1>}
     uVRegion {bIsValid=false, max=<0 0>, min=<0 0>}
    }
    renderTransform {angle=-40, scale=<-1 1>, shear=<0 0>, translation=<0 20>}
   }
   slot {class=OverlaySlot}
  }
  Widget_PreviewInfo_External {
   widgetClassPath=/FortniteCoreUI/Frontend/ItemPreview/WBP_ItemPreview_Info
   bIsVariable=true
   parent=Overlay_0
   widget=Widget_PreviewInfo_External
   slot {class=OverlaySlot, properties {horizontalAlignment=HAlign_Center, verticalAlignment=VAlign_Center}}
  }
  TabList_TopNav {
   widgetClassPath=&/WBP_ArmTale_WeaponDetails_TopNav
   bIsVariable=true
   parent=Overlay_13
   widget=TabList_TopNav
   slot {class=OverlaySlot, properties {horizontalAlignment=HAlign_Center, padding=<ltrb 0 40 0 0>}}
  }
 ]
}
