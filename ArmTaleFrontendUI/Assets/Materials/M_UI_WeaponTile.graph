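# Translucent UI material that draws a weapon tile for the ArmTale frontend, tinted by rarity colours.
# An offset radial rarity gradient forms the background, alongside a colour-dodged info panel, a rotated weapon icon with drop shadow and a two-colour bottom strip.
# SDF box functions produce the tile mask and outline, while rotated UV thresholds cut the antialiased panel and strip.
# Named reroutes feed these layers into a final lerp composite.
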Material /ArmTaleFrontendUI/Assets/Materials/M_UI_WeaponTile
graph {
 properties {
  bAutomaticallySetUsageInEditor=false
  bUsedWithStaticMesh=false
  blendMode=BLEND_Translucent
  materialDomain=MD_UI
  usedShadingModels=Default Lit
 }
 nodes [
  n1 = NamedRerouteUsage()
  n2 = VectorParameter() {parameterName=InfoPanelColor, defaultValue=<rgba 0.552 0.552 0.552 1>}
  n3 = LinearInterpolate(
   A=MaterialFunctionCall!(Blend=Multiply(A=@n1, B=@n2.RGB), Base=NamedRerouteUsage()) {materialFunction=/Engine/Functions/Engine_MaterialFunctions03/Blends/Blend_ColorDodge}.Result
   B=NamedRerouteUsage()
   Alpha=NamedRerouteUsage()
  )
  n4 = NamedRerouteUsage()
  n5 = NamedRerouteUsage()
  n6 = NamedRerouteUsage()
  n7 = NamedRerouteUsage()
  LinearInterpolate(A=Multiply(A=LinearInterpolate(A=@n3, B=@n4, Alpha=@n5), B=@n6), Alpha=@n7)
  Max(A=@n6, B=@n7)
  n8 = VectorParameter() {parameterName=RarityColor_3, defaultValue=<rgba 0 0.0976 0.521 1>}
  n9 = VectorParameter() {parameterName=RarityColor_2, defaultValue=<rgba 0 0.4287 1 1>}
  n10 = Subtract(
   A=MaterialFunctionCall() {materialFunction=&GetUserInterfaceUV}.Normalized UV
   B=AppendVector(
    A=ScalarParameter() {defaultValue=-0.25, parameterName=GradientOffsetX}
    B=ScalarParameter() {defaultValue=0.25, parameterName=GradientOffsetY}
   )
  )
  n11 = ScalarParameter() {defaultValue=2, parameterName=GradientScale}
  n12 = MaterialFunctionCall(UV=MaterialFunctionCall(UV=MaterialFunctionCall(UVs=@n10, Texture Scale=@n11) {materialFunction=&ScaleUVsByCenter}.UVs) {materialFunction=&MF_UI_UVScaling}.Scaled To Fill) {materialFunction=&MF_UI_RadialGradient}
  NamedRerouteDeclaration(None=Multiply(A=LinearInterpolate(A=@n8.RGB, B=@n9.RGB, Alpha=@n12.Smooth), B=ScalarParameter() {defaultValue=0.15, parameterName=RarityGradientOpacity}) {constB=0.2}) {name=BackgroundGradient, nodeColor=<rgba 0 0 0 1>}
  n13 = Add(A=MaterialFunctionCall!(UVs, Size=Constant() {r=1}, CornerRoundness, KeepAspectRatio=StaticBool() {value=true}, PixelSize) {materialFunction=/FortUILibrary/MaterialFunctions/Shapes/SDFs/MF_SDF_Box}.SDF) {constB=0.025}
  NamedRerouteDeclaration(None=MaterialFunctionCall(DistanceField=@n13) {materialFunction=&MF_UI_AntiAliasing}.Inner) {name=TileMask, nodeColor=<rgba 0 0 0 1>}
  NamedRerouteDeclaration(None=MaterialFunctionCall!(DistanceField=@n13, Threshold, Thickness=ScalarParameter() {parameterName=OutlineThickness}, Alignment, Smoothing) {materialFunction=/FortUILibrary/MaterialFunctions/Shapes/Modifiers/MF_UI_ShapeOutline}.Outline) {name=TileOutline, nodeColor=<rgba 0 0 0 1>}
  n14 = MaterialFunctionCall() {materialFunction=&GetUserInterfaceUV}
  n15 = Rotator(Coordinate=@n14.Normalized UV, Time=ScalarParameter() {defaultValue=3, parameterName=InfoPanelRotation}) {speed=0.0175}
  NamedRerouteDeclaration(None=MaterialFunctionCall(DistanceField=ComponentMask(None=@n15) {g=true}, Threshold=ScalarParameter() {defaultValue=0.5, parameterName=InfoPanelPosition}) {materialFunction=&MF_UI_AntiAliasing}.Outer) {name=InfoPanelMask, nodeColor=<rgba 0 0 0 1>}
  n16 = MaterialFunctionCall!(
   UVs=Subtract(
    A=MaterialFunctionCall() {materialFunction=&GetUserInterfaceUV}.Normalized UV
    B=Constant2Vector() {g=-0.2, r=-0.025}
   )
   IconTexture=TextureObjectParameter() {parameterName=IconTexture, texture=/Game/UI/Foundation/Textures/Icons/Weapons/Items/T-Icon-Weapons-PistolMark2TT}
   IconScale=ScalarParameter() {defaultValue=0.65, parameterName=IconScale}
   IconRotation=ScalarParameter() {defaultValue=-35, parameterName=IconRotation}
   ShadowOffset
   ShadowOpacity
  ) {materialFunction=/ArmTaleFrontendUI/Shared/MF_UI_IconWithDropShadow}
  NamedRerouteDeclaration(None=@n16.RGB) {name=WeaponIconColor, nodeColor=<rgba 0 0 0 1>}
  NamedRerouteDeclaration(None=@n16.A) {name=WeaponIconMask, nodeColor=<rgba 0 0 0 1>}
  NamedRerouteDeclaration(None=LinearInterpolate(
    A=VectorParameter() {parameterName=RarityColor_1, defaultValue=<rgba 0 1 0.9647 1>}.RGB
    B=VectorParameter() {parameterName=RarityColor_2, defaultValue=<rgba 0 0.4287 1 1>}.RGB
    Alpha=ComponentMask(None=@n14.Normalized UV) {r=true}
   )) {name=BottomStripColor, nodeColor=<rgba 0 0 0 1>}
  n17 = Rotator(Coordinate=@n14.Normalized UV, Time=ScalarParameter() {defaultValue=-1.5, parameterName=BottomStripRotation}) {speed=0.0175}
  NamedRerouteDeclaration(None=MaterialFunctionCall(DistanceField=ComponentMask(None=@n17) {g=true}, Threshold=ScalarParameter() {defaultValue=0.95, parameterName=BottomStripPosition}) {materialFunction=&MF_UI_AntiAliasing}.Outer) {name=BottomStripMask, nodeColor=<rgba 0 0 0 1>}
 ]
}
