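# Weapon tile list entry for the ArmTale frontend, showing name, rarity, level, progress and equipped mods.
# The list item is cast to a FAKE weapon viewmodel, and Handle functions rebuild mod slots and set the progress material.
# The rarity handler derives level text from the rarity enum and feeds RarityLabel's returned colors to the tile material.
# Focus animates scale and material parameters; hover animates the gradient only without focused descendants.
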WidgetBlueprint /ArmTaleFrontendUI/Widgets/WBP_ArmTale_WeaponListEntry
animations [
 AnimHovered {
  binding_count=1
  length_seconds=0.15
  object=&self:AnimHovered
  sequence {bindings [Image_WeaponTile {
     id=#1
     possessed_class=FortLazyImage
     tracks [{
       type=MovieSceneWidgetMaterialTrack
       sections [{
         type=MovieSceneParameterSection
         end_frame=3
         start_frame=0
         channels [RarityGradientOpacity {type=MovieSceneScriptingFloatChannel, keys [[0,0.15],[3,0.2]]}]
        }]
      }]
    }]}
 }
 AnimFocused {
  binding_count=2
  length_seconds=0.15
  object=&self:AnimFocused
  sequence {bindings [
    Overlay_17 {
     id=#2
     possessed_class=Overlay
     tracks [{
       type=MovieScene2DTransformTrack
       sections [{
         type=MovieScene2DTransformSection
         end_frame=3
         start_frame=0
         channels [
          Translation.X {type=MovieSceneScriptingFloatChannel, default=0}
          Translation.Y {type=MovieSceneScriptingFloatChannel, default=0}
          Angle {type=MovieSceneScriptingFloatChannel, default=0}
          Scale.X {type=MovieSceneScriptingFloatChannel, keys [[0,1],[3,1.05]]}
          Scale.Y {type=MovieSceneScriptingFloatChannel, keys [[0,1],[3,1.05]]}
          Shear.X {type=MovieSceneScriptingFloatChannel, default=0}
          Shear.Y {type=MovieSceneScriptingFloatChannel, default=0}
         ]
        }]
      }]
    }
    Image_WeaponTile {
     id=#3
     possessed_class=FortLazyImage
     tracks [{
       type=MovieSceneWidgetMaterialTrack
       sections [{
         type=MovieSceneParameterSection
         end_frame=3
         start_frame=0
         channels [
          IconScale {type=MovieSceneScriptingFloatChannel, keys [[0,0.65],[3,0.7]]}
          IconRotation {type=MovieSceneScriptingFloatChannel, keys [[0,-35],[3,-30]]}
          RarityGradientOpacity {type=MovieSceneScriptingFloatChannel, keys [[0,0.15],[3,0.3]]}
          OutlineThickness {type=MovieSceneScriptingFloatChannel, keys [[0,0],[3,0.025]]}
         ]
        }]
      }]
    }
   ]}
 }
]
graphs [
 EventGraph {nodes [
   Event On Hovered()
   n1 = HasAnyUserFocusedDescendants!(self:Widget Object Reference) -> (ReturnValue:bool)
   > n2 = Branch(Condition=@n1)
   n3 = Get AnimHovered() -> (AnimHovered:Widget Animation Object Reference)
   PlayAnimationForward(execute=@n2.else, InAnimation=@n3)
   Event On Unfocused() {is=title K2Node_Event}
   n4 = Get AnimFocused() -> (AnimFocused:Widget Animation Object Reference)
   > PlayAnimationReverse(InAnimation=@n4)
   Event On Unhovered()
   > n5 = Branch(Condition=@n1)
   PlayAnimationReverse(execute=@n5.else, InAnimation=@n3)
   n6 = Event OnListItemObjectSet()
   > n7 = Cast To VM_ArmTale_WeaponEntry_FAKE!(execute, Object=@n6.ListItemObject) -> (then, CastFailed, AsVM Arm Tale Weapon Entry FAKE:VM Arm Tale Weapon Entry FAKE Object Reference) {is=title K2Node_DynamicCast}
   > Set VM_ArmTale_WeaponEntry_FAKE(VM_ArmTale_WeaponEntry_FAKE:VM Arm Tale Weapon Entry FAKE Object Reference=@n7) -> (Output_Get:VM Arm Tale Weapon Entry FAKE Object Reference)
   Event On Focused() {is=title K2Node_Event}
   > PlayAnimationForward(InAnimation=@n4)
  ]}
 HandleEquippedModsChanged {nodes [
   HandleEquippedModsChanged!() -> (then, EquippedMods:Array of VM Arm Tale Mod Entry FAKE Object References) {is=title K2Node_FunctionEntry}
   > Reset!(
    execute
    self:Dynamic Entry Box Object Reference=Get EntryBox_ModSlots() -> (EntryBox_ModSlots:Dynamic Entry Box Object Reference)
    bDeleteWidgets:bool=false
   ) -> (then)
   > n1 = For Each Loop(Array:Array of VM Arm Tale Mod Entry FAKE Object References=Get EquippedMods() -> (EquippedMods:Array of VM Arm Tale Mod Entry FAKE Object References)) -> (Array Element:VM Arm Tale Mod Entry FAKE Object Reference)
   > n2 = Create Entry!(execute, self:Dynamic Entry Box Object Reference=Get EntryBox_ModSlots()) -> (then, ReturnValue:User Widget Object Reference)
   > n3 = Cast To WBP_ArmTale_ModSlot_Compact!(execute, Object=@n2) -> (then, CastFailed, AsWBP Arm Tale Mod Slot Compact:WBP Arm Tale Mod Slot Compact Object Reference) {is=title K2Node_DynamicCast}
   > SetVM_ArmTale_ModEntry_FAKE!(execute, self:WBP Arm Tale Mod Slot Compact Object Reference=@n3, Viewmodel:VM Arm Tale Mod Entry FAKE Object Reference=@n1) -> (then)
  ]}
 HandleLevelProgressChanged {nodes [
   n1 = HandleLevelProgressChanged!() -> (then, LevelProgress:double=~) {is=title K2Node_FunctionEntry}
   > n2 = GetDynamicMaterial(self=Get Image_ItemLevelProgress() -> (Image_ItemLevelProgress:Image Object Reference))
   > SetScalarParameterValue(self=@n2, ParameterName=Progress, Value=@n1)
  ]}
 HandleRarityChanged {nodes [
   n1 = HandleRarityChanged!() -> (then, Rarity:EFortRarity Enum=Common) {is=title K2Node_FunctionEntry}
   n2 = "To Integer (Byte)"!(InByte:Byte=@n1) -> (ReturnValue:int=0)
   > "SetText (Text)"(
    self=Get TextBlock_ItemLevel() -> (TextBlock_ItemLevel:Common Text Object Reference)
    InText=Format Text!(Format="LOCTEXT(\"LVL {LevelNum}\")", LevelNum:int=int + int(A=@n2)) -> (Result)
   )
   > n3 = SetRarity!(
    execute
    self:WBP Arm Tale Rarity Label Object Reference=Get RarityLabel() -> (RarityLabel:WBP Arm Tale Rarity Label Object Reference)
    Rarity:EFortRarity Enum=Get Rarity() -> (Rarity:EFortRarity Enum=Common)
   ) -> (
    then
    RarityData_Name:Text
    RarityData_Color1:Linear Color Structure
    RarityData_Color2:Linear Color Structure
    RarityData_Color3:Linear Color Structure
    RarityData_Color4:Linear Color Structure
    RarityData_Color5:Linear Color Structure
    RarityData_Radius:float=~
    RarityData_Falloff:float=~
    RarityData_Brightness:float=~
    RarityData_Roughness:float=~
    RarityData_Glow:float=~
   )
   > n4 = GetDynamicMaterial(self=Get Image_WeaponTile() -> (Image_WeaponTile:Fort Lazy Image Object Reference))
   > SetVectorParameterValue(self=@n4, ParameterName=RarityColor_1, Value=@n3.RarityData_Color1)
   > SetVectorParameterValue(self=@n4, ParameterName=RarityColor_2, Value=@n3.RarityData_Color2)
   > SetVectorParameterValue(self=@n4, ParameterName=RarityColor_3, Value=@n3.RarityData_Color3)
  ]}
]
variables []
widgets {
 info {inheritedWidgetCount=2, parentClass=/Script/FortniteUI.FortItemListEntryButton, rootWidgetClass=&Overlay, widgetCount=13}
 widgets [
  Overlay_17 {widgetClassPath=&Overlay, widget=Overlay_17, properties {pixelSnapping=Disabled}}
  Image_WeaponTile {
   widgetClassPath=&FortLazyImage
   bIsVariable=true
   parent=Overlay_17
   widget=Image_WeaponTile
   properties {
    materialTextureParamName=IconTexture
    brush {
     bIsDynamicallyLoaded=false
     drawAs=Image
     imageType=FullColor
     mirroring=NoMirror
     resourceName=None
     resourceObject=/ArmTaleFrontendUI/Assets/Materials/M_UI_WeaponTile
     tiling=NoTile
     imageSize=<32 32>
     margin=<ltrb 0 0 0 0>
     outlineSettings {
      bUseBrushTransparency=false
      roundingType=HalfHeightRadius
      width=0
      color {colorUseRule=UseColor_Specified, specifiedColor=<rgba 0 0 0 0>}
      cornerRadii=<0 0 0 0>
     }
     tintColor {colorUseRule=UseColor_Specified, specifiedColor=<rgba 1 1 1 1>}
     uVRegion {bIsValid=false, max=<0 0>, min=<0 0>}
    }
   }
   slot {class=OverlaySlot, properties {horizontalAlignment=HAlign_Fill, verticalAlignment=VAlign_Fill}}
  }
  VerticalBox_0 {
   widgetClassPath=&VerticalBox
   parent=Overlay_17
   widget=VerticalBox_0
   slot {class=OverlaySlot, properties {horizontalAlignment=HAlign_Fill, verticalAlignment=VAlign_Bottom, padding=<ltrb 20 20 20 36>}}
  }
  TextBlock_ItemName {
   widgetClassPath=&CommonTextBlock
   bIsVariable=true
   parent=VerticalBox_0
   widget=TextBlock_ItemName
   properties {
    clipping=ClipToBounds
    text="{Item Name}"
    textOverflowPolicy=Ellipsis
    textTransformPolicy=ToUpper
    font {
     bForceMonospaced=false
     bMaterialIsStencil=false
     fontMaterial=null
     fontObject=&BurbankSmall
     letterSpacing=-100
     monospacedWidth=1
     size=19.5
     skewAmount=0
     typefaceFontName=Black
     outlineSettings {
      bApplyOutlineToDropShadows=false
      bMiteredCorners=false
      bSeparateFillAlpha=false
      outlineMaterial=null
      outlineSize=0
      outlineColor=<rgba 0 0 0 1>
     }
    }
    shadowOffset=<0 0>
   }
   slot {class=VerticalBoxSlot}
  }
  RarityLabel {
   widgetClassPath=&/WBP_ArmTale_RarityLabel
   bIsVariable=true
   parent=VerticalBox_0
   widget=RarityLabel
   slot {class=VerticalBoxSlot, properties {horizontalAlignment=HAlign_Left, verticalAlignment=VAlign_Center}}
  }
  Spacer_3 {widgetClassPath=&Spacer, parent=VerticalBox_0, widget=Spacer_3, properties {size=<14 12>}, slot {class=VerticalBoxSlot}}
  HorizontalBox_0 {widgetClassPath=&HorizontalBox, parent=VerticalBox_0, widget=HorizontalBox_0, slot {class=VerticalBoxSlot}}
  TextBlock_ItemLevel {
   widgetClassPath=&CommonTextBlock
   bIsVariable=true
   parent=HorizontalBox_0
   widget=TextBlock_ItemLevel
   properties {
    text="{LVL 2}"
    textTransformPolicy=ToUpper
    font {
     bForceMonospaced=false
     bMaterialIsStencil=false
     fontMaterial=null
     fontObject=&HeadingNow
     letterSpacing=0
     monospacedWidth=1
     size=12
     skewAmount=0
     typefaceFontName=75 Medium
     outlineSettings {
      bApplyOutlineToDropShadows=false
      bMiteredCorners=false
      bSeparateFillAlpha=false
      outlineMaterial=null
      outlineSize=0
      outlineColor=<rgba 0 0 0 1>
     }
    }
    shadowOffset=<0 0>
   }
   slot {class=HorizontalBoxSlot, properties {verticalAlignment=VAlign_Center, size {sizeRule=Fill, value=1}}}
  }
  Spacer_2 {widgetClassPath=&Spacer, parent=HorizontalBox_0, widget=Spacer_2, properties {size=<4 4>}, slot {class=HorizontalBoxSlot}}
  EntryBox_ModSlots {
   widgetClassPath=&DynamicEntryBox
   bIsVariable=true
   parent=HorizontalBox_0
   widget=EntryBox_ModSlots
   properties {
    entryWidgetClass=&/WBP_ArmTale_ModSlot_Compact.WBP_ArmTale_ModSlot_Compact_C
    numDesignerPreviewEntries=4
    entrySpacing=<4 0>
   }
   slot {class=HorizontalBoxSlot}
  }
  Image_ItemLevelProgress {
   widgetClassPath=&Image
   bIsVariable=true
   parent=VerticalBox_0
   widget=Image_ItemLevelProgress
   properties {brush {
     bIsDynamicallyLoaded=false
     drawAs=Image
     imageType=FullColor
     mirroring=NoMirror
     resourceName=None
     resourceObject=/ArmTaleFrontendUI/Assets/Materials/M_UI_WeaponLevel_ProgressBar
     tiling=NoTile
     imageSize=<8 8>
     margin=<ltrb 0 0 0 0>
     outlineSettings {
      bUseBrushTransparency=false
      roundingType=HalfHeightRadius
      width=0
      color {colorUseRule=UseColor_Specified, specifiedColor=<rgba 0 0 0 0>}
      cornerRadii=<0 0 0 0>
     }
     tintColor {colorUseRule=UseColor_Specified, specifiedColor=<rgba 1 1 1 1>}
     uVRegion {bIsValid=false, max=<0 0>, min=<0 0>}
    }}
   slot {class=VerticalBoxSlot, properties {padding=<ltrb 1 4 0 0>}}
  }
  ItemWidget {widgetClassPath=/Script/FortniteUI.FortMultiSizeItemCard, bInherited=true, widget=null}
  InputActionWidget {widgetClassPath=&CommonActionWidget, bInherited=true, widget=null}
 ]
}
